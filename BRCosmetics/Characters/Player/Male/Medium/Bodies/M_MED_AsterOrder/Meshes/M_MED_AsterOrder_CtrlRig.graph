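# Control Rig for the M_MED_AsterOrder player body, applying procedural corrective deformation to its jacket, collar, thigh chains and necklace.
# After a shared DeformRig pass, leg pose weights come from anim attributes and spherical pose readers.
# Weighted parent constraints attach base bones to the body, then additive transforms adjust them and their chains.
# Leg pose, crouch and slide flags, spine and head rotation, and pickaxe-swing state weight these corrections.
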ControlRigBlueprint /BRCosmetics/Characters/Player/Male/Medium/Bodies/M_MED_AsterOrder/Meshes/M_MED_AsterOrder_CtrlRig
graphs [
 New Function_ContainedGraph {nodes [
   n1 = RigVMFunctionEntryNode()
   RigVMFunction_Sequence = &RigVMFunction_Sequence(ExecuteContext=@n1.ExecuteContext)
   &RigUnit_ParentConstraint(
    AdvancedSettings=(RotationOrderForFilter=XZY)
    Child=(Type=Bone,Name="base_Jacket_front_l")
    Parents=((Item=(Type=Bone,Name="spine_01"),Weight=0.004),(Item=(Type=Bone,Name="spine_05"),Weight=0.0624),(Item=(Type=Bone,Name="spine_02"),Weight=0.2104),(Item=(Type=Bone,Name="spine_03"),Weight=0.3452),(Item=(Type=Bone,Name="spine_04"),Weight=0.378))
    Weight=1
    ExecutePin=@RigVMFunction_Sequence.A
   )
   > &RigUnit_ParentConstraint(
    AdvancedSettings=(RotationOrderForFilter=XZY)
    Child=(Type=Bone,Name="base_Jacket_a_l")
    Parents=((Item=(Type=Bone,Name="spine_01"),Weight=0.0039),(Item=(Type=Bone,Name="spine_05"),Weight=0.0213),(Item=(Type=Bone,Name="spine_02"),Weight=0.1858),(Item=(Type=Bone,Name="spine_03"),Weight=0.3801),(Item=(Type=Bone,Name="spine_04"),Weight=0.409))
    Weight=1
   )
   > &RigUnit_ParentConstraint(
    AdvancedSettings=(RotationOrderForFilter=XZY)
    Child=(Type=Bone,Name="base_Jacket_b_l")
    Parents=((Item=(Type=Bone,Name="spine_01"),Weight=0.0003),(Item=(Type=Bone,Name="spine_02"),Weight=0.1484),(Item=(Type=Bone,Name="spine_04"),Weight=0.4102),(Item=(Type=Bone,Name="spine_03"),Weight=0.4411))
    Weight=1
   )
   > &RigUnit_ParentConstraint(
    AdvancedSettings=(RotationOrderForFilter=XZY)
    Child=(Type=Bone,Name="base_Jacket_c_l")
    Parents=((Item=(Type=Bone,Name="spine_01"),Weight=0.0014),(Item=(Type=Bone,Name="spine_05"),Weight=0.0374),(Item=(Type=Bone,Name="spine_02"),Weight=0.1803),(Item=(Type=Bone,Name="spine_04"),Weight=0.3685),(Item=(Type=Bone,Name="spine_03"),Weight=0.4124))
    Weight=1
   )
   > &RigUnit_ParentConstraint(
    AdvancedSettings=(RotationOrderForFilter=XZY)
    Child=(Type=Bone,Name="base_Jacket_back")
    Parents=((Item=(Type=Bone,Name="spine_01"),Weight=0.0018),(Item=(Type=Bone,Name="spine_05"),Weight=0.1321),(Item=(Type=Bone,Name="spine_02"),Weight=0.1465),(Item=(Type=Bone,Name="spine_03"),Weight=0.3592),(Item=(Type=Bone,Name="spine_04"),Weight=0.3604))
    Weight=1
   )
   > &RigUnit_ParentConstraint(
    AdvancedSettings=(RotationOrderForFilter=XZY)
    Child=(Type=Bone,Name="base_Jacket_front_r")
    Parents=((Item=(Type=Bone,Name="spine_01"),Weight=0.004),(Item=(Type=Bone,Name="spine_05"),Weight=0.0624),(Item=(Type=Bone,Name="spine_02"),Weight=0.2104),(Item=(Type=Bone,Name="spine_03"),Weight=0.3452),(Item=(Type=Bone,Name="spine_04"),Weight=0.378))
    Weight=1
   )
   > &RigUnit_ParentConstraint(
    AdvancedSettings=(RotationOrderForFilter=XZY)
    Child=(Type=Bone,Name="base_Jacket_a_r")
    Parents=((Item=(Type=Bone,Name="spine_01"),Weight=0.0039),(Item=(Type=Bone,Name="spine_05"),Weight=0.0213),(Item=(Type=Bone,Name="spine_02"),Weight=0.1858),(Item=(Type=Bone,Name="spine_03"),Weight=0.3801),(Item=(Type=Bone,Name="spine_04"),Weight=0.409))
    Weight=1
   )
   > &RigUnit_ParentConstraint(
    AdvancedSettings=(RotationOrderForFilter=XZY)
    Child=(Type=Bone,Name="base_Jacket_b_r")
    Parents=((Item=(Type=Bone,Name="spine_01"),Weight=0.0003),(Item=(Type=Bone,Name="spine_02"),Weight=0.1484),(Item=(Type=Bone,Name="spine_04"),Weight=0.4102),(Item=(Type=Bone,Name="spine_03"),Weight=0.4412))
    Weight=1
   )
   > &RigUnit_ParentConstraint(
    AdvancedSettings=(RotationOrderForFilter=XZY)
    Child=(Type=Bone,Name="base_Jacket_c_r")
    Parents=((Item=(Type=Bone,Name="spine_01"),Weight=0.0014),(Item=(Type=Bone,Name="spine_05"),Weight=0.0374),(Item=(Type=Bone,Name="spine_02"),Weight=0.1803),(Item=(Type=Bone,Name="spine_04"),Weight=0.3685),(Item=(Type=Bone,Name="spine_03"),Weight=0.4124))
    Weight=1
   )
   > &RigUnit_ParentConstraint(
    AdvancedSettings=(RotationOrderForFilter=XZY)
    Child=(Type=Bone,Name="base_Neck_Jacket_a_l")
    Parents="((Item=(Type=Bone,Name=\"head\"),Weight=0.0059),(Item=(Type=Bone,Name=\"spine_03\"),Weight=0.0258),(Item=(Type=Bone,Name=\"neck_01\"),Weight=0.0525),(Item=(Type=Bone,Name=\"clavicle_l\"),Weight=0.0709),(Item=(Type=Bone,Name=\"spine_04\"),Weight=0.078),(Item=(Type=Bo ...(truncated)"
    Weight=1
   )
   > &RigUnit_ParentConstraint(
    AdvancedSettings=(RotationOrderForFilter=XZY)
    Child=(Type=Bone,Name="base_Neck_Jacket_b_l")
    Parents="((Item=(Type=Bone,Name=\"upperarm_twist_01_l\"),Weight=0.0056),(Item=(Type=Bone,Name=\"spine_04\"),Weight=0.0094),(Item=(Type=Bone,Name=\"spine_03\"),Weight=0.0121),(Item=(Type=Bone,Name=\"neck_01\"),Weight=0.0692),(Item=(Type=Bone,Name=\"head\"),Weight=0.0972),(Ite ...(truncated)"
    Weight=1
   )
   > &RigUnit_ParentConstraint(
    AdvancedSettings=(RotationOrderForFilter=XZY)
    Child=(Type=Bone,Name="base_Neck_Jacket_c")
    Parents=((Item=(Type=Bone,Name="spine_03"),Weight=0.0175),(Item=(Type=Bone,Name="head"),Weight=0.107),(Item=(Type=Bone,Name="neck_01"),Weight=0.1235),(Item=(Type=Bone,Name="neck_02"),Weight=0.292),(Item=(Type=Bone,Name="spine_05"),Weight=0.4601))
    Weight=1
   )
   > &RigUnit_ParentConstraint(
    AdvancedSettings=(RotationOrderForFilter=XZY)
    Child=(Type=Bone,Name="base_Neck_Jacket_a_r")
    Parents="((Item=(Type=Bone,Name=\"spine_03\"),Weight=0.0227),(Item=(Type=Bone,Name=\"head\"),Weight=0.0248),(Item=(Type=Bone,Name=\"neck_01\"),Weight=0.053),(Item=(Type=Bone,Name=\"spine_04\"),Weight=0.0724),(Item=(Type=Bone,Name=\"clavicle_r\"),Weight=0.0801),(Item=(Type=Bo ...(truncated)"
    Weight=1
   )
   > &RigUnit_ParentConstraint(
    AdvancedSettings=(RotationOrderForFilter=XZY)
    Child=(Type=Bone,Name="base_Neck_Jacket_b_r")
    Parents="((Item=(Type=Bone,Name=\"upperarm_twist_01_r\"),Weight=0.012),(Item=(Type=Bone,Name=\"spine_04\"),Weight=0.0125),(Item=(Type=Bone,Name=\"spine_03\"),Weight=0.0139),(Item=(Type=Bone,Name=\"neck_01\"),Weight=0.0595),(Item=(Type=Bone,Name=\"head\"),Weight=0.1283),(Item ...(truncated)"
    Weight=1
   )
   > &RigUnit_ParentConstraint(
    AdvancedSettings=(RotationOrderForFilter=XZY)
    Child=(Type=Bone,Name="base_thigh_chain_a_l")
    Parents=((Item=(Type=Bone,Name="spine_03"),Weight=0.0009),(Item=(Type=Bone,Name="spine_02"),Weight=0.0072),(Item=(Type=Bone,Name="thigh_twist_01_l"),Weight=0.0542),(Item=(Type=Bone,Name="spine_01"),Weight=0.249),(Item=(Type=Bone,Name="pelvis"),Weight=0.6887))
    Weight=1
   )
   > &RigUnit_ParentConstraint(
    AdvancedSettings=(RotationOrderForFilter=XZY)
    Child=(Type=Bone,Name="base_thigh_chain_b_l")
    Parents=((Item=(Type=Bone,Name="thigh_twist_01_l"),Weight=0.057),(Item=(Type=Bone,Name="spine_01"),Weight=0.0787),(Item=(Type=Bone,Name="pelvis"),Weight=0.8643))
    Weight=1
   )
   > &RigUnit_ParentConstraint(
    AdvancedSettings=(RotationOrderForFilter=XZY)
    Child=(Type=Bone,Name="base_thigh_chain_b_r")
    Parents=((Item=(Type=Bone,Name="thigh_r"),Weight=0.0003),(Item=(Type=Bone,Name="spine_01"),Weight=0.0106),(Item=(Type=Bone,Name="pelvis"),Weight=0.351),(Item=(Type=Bone,Name="thigh_twist_01_r"),Weight=0.638))
    Weight=1
   )
   > &RigUnit_ParentConstraint(
    AdvancedSettings=(RotationOrderForFilter=XZY)
    Child=(Type=Bone,Name="base_thigh_chain_a_r")
    Parents=((Item=(Type=Bone,Name="spine_03"),Weight=0.0032),(Item=(Type=Bone,Name="thigh_twist_01_r"),Weight=0.0305),(Item=(Type=Bone,Name="spine_01"),Weight=0.1008),(Item=(Type=Bone,Name="pelvis"),Weight=0.8655))
    Weight=1
   )
   > &RigUnit_ParentConstraint(
    AdvancedSettings=(RotationOrderForFilter=XZY)
    Child=(Type=Bone,Name="base_thigh_chain_ring_l")
    Parents="((Item=(Type=Bone,Name=\"spine_04\"),Weight=0.0008),(Item=(Type=Bone,Name=\"thigh_twist_01_r\"),Weight=0.0035),(Item=(Type=Bone,Name=\"thigh_twist_01_l\"),Weight=0.0185),(Item=(Type=Bone,Name=\"spine_03\"),Weight=0.0347),(Item=(Type=Bone,Name=\"deform_groin_l\"),Wei ...(truncated)"
    Weight=1
   )
   > &RigUnit_ParentConstraint(
    AdvancedSettings=(RotationOrderForFilter=XZY)
    Child=(Type=Bone,Name="base_neck_c")
    Parents="((Item=(Type=Bone,Name=\"FACIAL_C_Jaw\"),Weight=0.0056),(Item=(Type=Bone,Name=\"FACIAL_R_UnderChin\"),Weight=0.0122),(Item=(Type=Bone,Name=\"spine_05\"),Weight=0.018),(Item=(Type=Bone,Name=\"head\"),Weight=0.0989),(Item=(Type=Bone,Name=\"neck_01\"),Weight=0.1523),(I ...(truncated)"
    Weight=1
   )
   RigVMFunctionReturnNode(ExecuteContext=@RigVMFunction_Sequence.B)
  ]}
 New Function_ContainedGraph {nodes [
   n1 = RigVMFunctionEntryNode()
   RigVMFunction_MathQuaternionToEuler = &RigVMFunction_MathQuaternionToEuler(
    RotationOrder=ZYX
    Value=&RigUnit_GetRelativeTransformForItem(
     Child=(Type=Bone,Name="head")
     Parent=(Type=Bone,Name="neck_01")
     bChildInitial=False
     bParentInitial=False
     Child.Name=@n1.Child
     Parent.Name=@n1.Parent
    ).RelativeTransform.Rotation
   )
   Remap = &RigVMFunction_MathDoubleRemap(
    TargetMaximum=1
    TargetMinimum=0
    SourceMaximum=@n1.Source Max
    SourceMinimum=@n1.Source Min
    Value=@RigVMFunction_MathQuaternionToEuler.Result.Z
   )
   > RigVMFunctionReturnNode(
    Delta_z=@Remap.Result
    Delta_X_Negative=&RigVMFunction_MathDoubleRemap(SourceMaximum=-60, SourceMinimum=-1, TargetMaximum=1, TargetMinimum=0, Value=@RigVMFunction_MathQuaternionToEuler.Result.X).Result
    Delta_X_Positive=&RigVMFunction_MathDoubleRemap(SourceMaximum=60, SourceMinimum=0, TargetMaximum=1, TargetMinimum=0, Value=@RigVMFunction_MathQuaternionToEuler.Result.X).Result
   )
  ]}
 New Function_ContainedGraph {nodes [
   DeltaRotation = RigVMFunctionReferenceNode(Child=Head, ExecuteContext=(), Parent=neck_01, Source Max=-60, Source Min=-10)
   > &RigUnit_ModifyTransforms(
    ItemToModify="((Item=(Type=Bone,Name=\"None\"),Transform=(Rotation=(X=0,Y=0,Z=0,W=1),Translation=(X=0,Y=0,Z=0),Scale3D=(X=1,Y=1,Z=1))),(Item=(Type=Bone,Name=\"dyn_Neck_Jacket_a_l\"),Transform=(Rotation=(X=0,Y=-0.1736,Z=0,W=0.9848),Translation=(X=0,Y=0,Z=0),Scale3D=(X=1,Y=1, ...(truncated)"
    WeightMinimum=0
    Weight=@DeltaRotation.Delta_z
   )
   > &RigUnit_ModifyTransforms(
    ItemToModify="((Item=(Type=Bone,Name=\"None\"),Transform=(Rotation=(X=0,Y=0,Z=0,W=1),Translation=(X=0,Y=0,Z=0),Scale3D=(X=1,Y=1,Z=1))),(Item=(Type=Bone,Name=\"dyn_Neck_Jacket_a_r\"),Transform=(Rotation=(X=0.1277,Y=0.1449,Z=-0.2685,W=0.9437),Translation=(X=0,Y=0,Z=0),Scale3D ...(truncated)"
    WeightMinimum=0
    Weight=@DeltaRotation.Delta_z
   )
   > &RigUnit_ModifyTransforms(
    ItemToModify="((Item=(Type=Bone,Name=\"None\"),Transform=(Rotation=(X=0,Y=0,Z=0,W=1),Translation=(X=0,Y=0,Z=0),Scale3D=(X=1,Y=1,Z=1))),(Item=(Type=Bone,Name=\"dyn_Neck_Jacket_c\"),Transform=(Rotation=(X=0.4226,Y=0,Z=0,W=0.9063),Translation=(X=0,Y=0,Z=0),Scale3D=(X=1,Y=1,Z=1 ...(truncated)"
    WeightMinimum=0
    Weight=@DeltaRotation.Delta_z
   )
   n1 = RigVMFunctionEntryNode()
   RigVMFunction_Sequence = &RigVMFunction_Sequence(ExecuteContext=@n1.ExecuteContext)
   n2 = &RigUnit_SphericalPoseReader(
    ActiveRegionScaleFactors=(PositiveWidth=2.3,NegativeWidth=2.3,PositiveHeight=2.3,NegativeHeight=2.3)
    ActiveRegionSize=0.05
    Debug=(bDrawDebug=false,DebugScale=25,DebugSegments=20,DebugThickness=0.25)
    DriverAxis=(X=0,Y=1,Z=0)
    DriverItem=(Type=Bone,Name="head")
    FalloffRegionScaleFactors=(PositiveWidth=0.5,NegativeWidth=0.5,PositiveHeight=0.5,NegativeHeight=0.5)
    FalloffSize=0.3
    FlipHeightScaling=false
    FlipWidthScaling=false
    OptionalParentItem=(Type=Bone)
    RotationOffset=(X=50,Y=0,Z=0)
    ExecutePin=@RigVMFunction_Sequence.A
   )
   n3 = RigVMVariableNode() {variable=Pub_IsPickaxeSwing}
   Multiply = &RigVMFunction_MathDoubleMul(A=@n2.OutputParam, B=@n3.Value)
   > &RigUnit_ModifyTransforms(
    ItemToModify="((Item=(Type=Bone,Name=\"None\"),Transform=(Rotation=(X=0,Y=0,Z=0,W=1),Translation=(X=0,Y=0,Z=0),Scale3D=(X=1,Y=1,Z=1))),(Item=(Type=Bone,Name=\"dyn_Neck_Jacket_b_r\"),Transform=(Rotation=(X=0,Y=0,Z=0,W=1),Translation=(X=-6,Y=0,Z=0),Scale3D=(X=1,Y=1,Z=1))),(It ...(truncated)"
    WeightMinimum=0
    Weight=@Multiply.Result
   )
   > n4 = &RigUnit_SphericalPoseReader(
    ActiveRegionScaleFactors=(PositiveWidth=2.3,NegativeWidth=2.3,PositiveHeight=2.3,NegativeHeight=2.3)
    ActiveRegionSize=0.05
    Debug=(bDrawDebug=false,DebugScale=25,DebugSegments=20,DebugThickness=0.25)
    DriverAxis=(X=0,Y=1,Z=0)
    DriverItem=(Type=Bone,Name="head")
    FalloffRegionScaleFactors=(PositiveWidth=0.5,NegativeWidth=0.5,PositiveHeight=0.5,NegativeHeight=0.5)
    FalloffSize=0.3
    FlipHeightScaling=false
    FlipWidthScaling=false
    OptionalParentItem=(Type=Bone)
    RotationOffset=(X=140,Y=0,Z=0)
   )
   > &RigUnit_ModifyTransforms(
    ItemToModify="((Item=(Type=Bone,Name=\"None\"),Transform=(Rotation=(X=0,Y=0,Z=0,W=1),Translation=(X=0,Y=0,Z=0),Scale3D=(X=1,Y=1,Z=1))),(Item=(Type=Bone,Name=\"dyn_Neck_Jacket_a_l\"),Transform=(Rotation=(X=0,Y=0.1736,Z=0,W=0.9848),Translation=(X=1,Y=0,Z=0),Scale3D=(X=1,Y=1,Z ...(truncated)"
    WeightMinimum=0
    Weight=&RigVMFunction_MathDoubleMul(A=@n4.OutputParam, B=@n3.Value).Result
   )
   RigVMFunctionReturnNode(ExecuteContext=@RigVMFunction_Sequence.B)
  ]}
 New Function_ContainedGraph {nodes [
   n1 = RigVMFunctionEntryNode()
   RigVMFunction_Sequence = &RigVMFunction_Sequence(ExecuteContext=@n1.ExecuteContext)
   RigVMVariableNode(
    Value=RigVMDispatchNode(BoneName=root, Default=0, Name=deformRig_l_thigh_up_alpha) {template="DISPATCH_RigDispatch_GetAnimAttribute(in Name,in BoneName,in Default,out Value,out Found)"}.Value
    ExecuteContext=@RigVMFunction_Sequence.A
   ) {variable=SPR L Leg Fwd}
   > RigVMVariableNode(Value=RigVMDispatchNode(BoneName=root, Default=0, Name=deformRig_r_thigh_up_alpha) {template="DISPATCH_RigDispatch_GetAnimAttribute(in Name,in BoneName,in Default,out Value,out Found)"}.Value) {variable=SPR R Leg Fwd}
   > n2 = &RigUnit_SphericalPoseReader(
    ActiveRegionScaleFactors=(PositiveWidth=3.5,NegativeWidth=0.5,PositiveHeight=3,NegativeHeight=3)
    ActiveRegionSize=0.05
    Debug=(bDrawDebug=false,DebugScale=25,DebugSegments=20,DebugThickness=0.25)
    DriverAxis=(X=-1,Y=0,Z=0)
    DriverItem=(Type=Bone,Name="thigh_l")
    FalloffRegionScaleFactors=(PositiveWidth=1.35,NegativeWidth=1,PositiveHeight=1,NegativeHeight=1)
    FalloffSize=0.3
    FlipHeightScaling=false
    FlipWidthScaling=false
    OptionalParentItem=(Type=Bone)
    RotationOffset=(X=90,Y=0,Z=180)
   )
   > RigVMVariableNode(Value=@n2.OutputParam) {variable=SPR L Leg Bkd}
   > n3 = &RigUnit_SphericalPoseReader(
    ActiveRegionScaleFactors=(PositiveWidth=3.5,NegativeWidth=0.5,PositiveHeight=3,NegativeHeight=3)
    ActiveRegionSize=0.05
    Debug=(bDrawDebug=false,bDrawLocalAxes=false,DebugScale=25,DebugSegments=20,DebugThickness=0.25)
    DriverAxis=(X=1,Y=0,Z=0)
    DriverItem=(Type=Bone,Name="thigh_r")
    FalloffRegionScaleFactors=(PositiveWidth=1.35,NegativeWidth=1,PositiveHeight=1,NegativeHeight=1)
    FalloffSize=0.3
    FlipHeightScaling=false
    FlipWidthScaling=true
    OptionalParentItem=(Type=Bone)
    RotationOffset=(X=-90,Y=0,Z=180)
   )
   > RigVMVariableNode(Value=@n3.OutputParam) {variable=SPR R Leg Bkd}
   > n4 = &RigUnit_SphericalPoseReader(
    ActiveRegionScaleFactors=(PositiveWidth=2,NegativeWidth=2,PositiveHeight=2,NegativeHeight=2)
    ActiveRegionSize=0.05
    Debug=(bDrawDebug=false,bDraw2D=false,DebugScale=25,DebugSegments=25,DebugThickness=0.25)
    DriverAxis=(X=-1,Y=0,Z=0)
    DriverItem=(Type=Bone,Name="thigh_l")
    FalloffRegionScaleFactors=(PositiveWidth=0.2,NegativeWidth=1.2,PositiveHeight=0.5,NegativeHeight=1)
    FalloffSize=0.5
    FlipHeightScaling=false
    FlipWidthScaling=false
    OptionalParentItem=(Type=Bone)
    RotationOffset=(X=180,Y=19,Z=0)
   )
   > RigVMVariableNode(Value=@n4.OutputParam) {variable=SPR L Leg Otw}
   > n5 = &RigUnit_SphericalPoseReader(
    ActiveRegionScaleFactors=(PositiveWidth=4,NegativeWidth=7,PositiveHeight=2,NegativeHeight=2)
    ActiveRegionSize=0.05
    Debug=(bDrawDebug=false,bDraw2D=false,DebugScale=25,DebugSegments=25,DebugThickness=0.25)
    DriverAxis=(X=1,Y=0,Z=0)
    DriverItem=(Type=Bone,Name="thigh_r")
    FalloffRegionScaleFactors=(PositiveWidth=1,NegativeWidth=1,PositiveHeight=0.5,NegativeHeight=0.7)
    FalloffSize=0.61
    FlipHeightScaling=false
    FlipWidthScaling=false
    OptionalParentItem=(Type=Bone)
    RotationOffset=(X=0,Y=19,Z=0)
   )
   > RigVMVariableNode(Value=@n5.OutputParam) {variable=SPR R Leg Otw}
   n6 = &RigVMFunction_AnimEvalRichCurve(
    Curve="(EditorCurveData=(Keys=((InterpMode=RCIM_Cubic,TangentMode=RCTM_Break,TangentWeightMode=RCTWM_WeightedNone,Time=0,Value=0,ArriveTangent=0,ArriveTangentWeight=0,LeaveTangent=0,LeaveTangentWeight=0),(InterpMode=RCIM_Cubic,TangentMode=RCTM_Break,TangentWeight ...(truncated)"
    SourceMinimum=0.5
    TargetMinimum=0
    Value=&RigVMFunction_MathDoubleRemap(
     SourceMaximum=30
     SourceMinimum=60
     TargetMaximum=0
     TargetMinimum=1
     bClamp=False
     Value=&RigUnit_GetRelativeTransformForItem(Child=(Type=Bone,Name="pelvis"), Parent=(Type=Bone,Name="root"), bChildInitial=False, bParentInitial=False).RelativeTransform.Translation.Z
    ).Result
   )
   > RigVMVariableNode(Value=@n6.Result) {variable=PelvisProx}
   RigVMFunctionReturnNode(ExecuteContext=@RigVMFunction_Sequence.B)
  ]}
 New Function_ContainedGraph {nodes [
   n1 = RigVMFunctionEntryNode()
   RigVMFunction_ControlFlowBranch = &RigVMFunction_ControlFlowBranch(Condition=@n1.Debug Controllers, ExecuteContext=@n1.ExecuteContext)
   RigVMFunctionReturnNode(ExecuteContext=@RigVMFunction_ControlFlowBranch.Completed)
   RigUnit_GetTransform = &RigUnit_GetTransform(Item=(Type=Control,Name="thigh_l_ctrl"), bInitial=False)
   &RigUnit_SetTransform(
    Item=(Type=Bone,Name="thigh_l")
    Weight=1
    bInitial=False
    Value=@RigUnit_GetTransform.Transform
    ExecutePin=@RigVMFunction_ControlFlowBranch.True
   )
   > RigUnit_SetTransform = &RigUnit_SetTransform(
    Item=(Type=Bone,Name="thigh_r")
    Weight=1
    bInitial=False
    Value=&RigUnit_GetTransform(Item=(Type=Control,Name="thigh_r_ctrl"), bInitial=False).Transform
   )
  ]}
 New Function_ContainedGraph {nodes [
   n1 = RigVMFunctionEntryNode()
   RigVMFunction_Sequence = &RigVMFunction_Sequence(ExecuteContext=@n1.ExecuteContext)
   &RigUnit_ModifyTransforms(
    ItemToModify="((Item=(Type=Bone,Name=\"None\"),Transform=(Rotation=(X=0,Y=0,Z=0,W=1),Translation=(X=0,Y=0,Z=0),Scale3D=(X=1,Y=1,Z=1))),(Item=(Type=Bone,Name=\"off_thigh_chain_a_l\"),Transform=(Rotation=(X=0.0233,Y=-0.6329,Z=-0.0999,W=0.7674),Translation=(X=-2.1642,Y=-0.2918 ...(truncated)"
    WeightMinimum=0
    ExecutePin=@RigVMFunction_Sequence.A
    Weight=RigVMVariableNode() {variable=SPR L Leg Fwd}.Value
   )
   > &RigUnit_ModifyTransforms(
    ItemToModify="((Item=(Type=Bone,Name=\"None\"),Transform=(Rotation=(X=0,Y=0,Z=0,W=1),Translation=(X=0,Y=0,Z=0),Scale3D=(X=1,Y=1,Z=1))),(Item=(Type=Bone,Name=\"off_thigh_chain_a_r\"),Transform=(Rotation=(X=-0.0935,Y=-0.5039,Z=-0.1104,W=0.8516),Translation=(X=0,Y=0,Z=0),Scale ...(truncated)"
    WeightMinimum=0
    Weight=RigVMVariableNode() {variable=SPR R Leg Fwd}.Value
   )
   > &RigUnit_ModifyTransforms(
    ItemToModify="((Item=(Type=Bone,Name=\"None\"),Transform=(Rotation=(X=0,Y=0,Z=0,W=1),Translation=(X=0,Y=0,Z=0),Scale3D=(X=1,Y=1,Z=1))),(Item=(Type=Bone,Name=\"off_thigh_chain_a_l\"),Transform=(Rotation=(X=0,Y=0,Z=0,W=1),Translation=(X=0,Y=0,Z=0),Scale3D=(X=1,Y=1,Z=1))),(Ite ...(truncated)"
    WeightMinimum=0
    Weight=RigVMVariableNode() {variable=SPR L Leg Bkd}.Value
   )
   > &RigUnit_ModifyTransforms(
    ItemToModify="((Item=(Type=Bone,Name=\"None\"),Transform=(Rotation=(X=0,Y=0,Z=0,W=1),Translation=(X=0,Y=0,Z=0),Scale3D=(X=1,Y=1,Z=1))),(Item=(Type=Bone,Name=\"off_thigh_chain_a_r\"),Transform=(Rotation=(X=0.0526,Y=-0.5484,Z=-0.1194,W=0.826),Translation=(X=0,Y=0,Z=0),Scale3D ...(truncated)"
    WeightMinimum=0
    Weight=RigVMVariableNode() {variable=SPR R Leg Bkd}.Value
   )
   > &RigUnit_ModifyTransforms(
    ItemToModify="((Item=(Type=Bone,Name=\"None\"),Transform=(Rotation=(X=0,Y=0,Z=0,W=1),Translation=(X=0,Y=0,Z=0),Scale3D=(X=1,Y=1,Z=1))),(Item=(Type=Bone,Name=\"off_thigh_chain_a_l\"),Transform=(Rotation=(X=-0.0324,Y=-0.2216,Z=-0.4579,W=0.8603),Translation=(X=1.3779,Y=-0.7356 ...(truncated)"
    WeightMinimum=0
    Weight=RigVMVariableNode() {variable=SPR L Leg Otw}.Value
   )
   > &RigUnit_ModifyTransforms(
    ItemToModify="((Item=(Type=Bone,Name=\"None\"),Transform=(Rotation=(X=0,Y=0,Z=0,W=1),Translation=(X=0,Y=0,Z=0),Scale3D=(X=1,Y=1,Z=1))),(Item=(Type=Bone,Name=\"off_thigh_chain_a_r\"),Transform=(Rotation=(X=0,Y=0,Z=-0.4536,W=0.8912),Translation=(X=0,Y=0,Z=0),Scale3D=(X=1,Y=1, ...(truncated)"
    WeightMinimum=0
    Weight=RigVMVariableNode() {variable=SPR R Leg Otw}.Value
   )
   > DeltaRotation = RigVMFunctionReferenceNode(Child=thigh_l, Parent=pelvis, Source Max=-120, Source Min=-90)
   > &RigUnit_ModifyTransforms(
    ItemToModify="((Item=(Type=Bone,Name=\"None\"),Transform=(Rotation=(X=0,Y=0,Z=0,W=1),Translation=(X=0,Y=0,Z=0),Scale3D=(X=1,Y=1,Z=1))),(Item=(Type=Bone,Name=\"off_thigh_chain_a_l\"),Transform=(Rotation=(X=0.2834,Y=-0.0613,Z=-0.2635,W=0.92),Translation=(X=-1.415,Y=0.3053,Z=1 ...(truncated)"
    WeightMinimum=0
    Weight=@DeltaRotation.Delta_z
   )
   > &RigUnit_ModifyTransforms(
    ItemToModify="((Item=(Type=Bone,Name=\"None\"),Transform=(Rotation=(X=0,Y=0,Z=0,W=1),Translation=(X=0,Y=0,Z=0),Scale3D=(X=1,Y=1,Z=1))),(Item=(Type=Bone,Name=\"off_thigh_chain_a_l\"),Transform=(Rotation=(X=0,Y=-0.0436,Z=0,W=0.999),Translation=(X=0,Y=0,Z=0),Scale3D=(X=1,Y=1,Z ...(truncated)"
    Weight=1
    WeightMinimum=0
   )
   RigVMFunctionReturnNode(ExecuteContext=@RigVMFunction_Sequence.B)
   &RigUnit_ModifyTransforms(
    ExecutePin=()
    ItemToModify="((Item=(Type=Bone,Name=\"None\"),Transform=(Rotation=(X=0,Y=0,Z=0,W=1),Translation=(X=0,Y=0,Z=0),Scale3D=(X=1,Y=1,Z=1))),(Item=(Type=Bone,Name=\"off_thigh_chain_a_l\"),Transform=(Rotation=(X=0.041,Y=-0.4831,Z=-0.094,W=0.8695),Translation=(X=-2.1642,Y=-0.2919,Z ...(truncated)"
    Weight=1
    WeightMinimum=0
   )
  ]}
 New Function_ContainedGraph {nodes [
   n1 = RigVMFunctionEntryNode()
   RigVMFunction_Sequence = &RigVMFunction_Sequence(ExecuteContext=@n1.ExecuteContext)
   RigVMFunction_MathQuaternionFromEuler = &RigVMFunction_MathQuaternionFromEuler(Euler=(X=20,Y=0,Z=0))
   n2 = RigVMVariableNode() {variable=Pub_IsCrouching}
   &RigUnit_ModifyTransforms(
    ItemToModify="((Item=(Type=Bone,Name=\"None\"),Transform=(Rotation=(X=0,Y=0,Z=0,W=1),Translation=(X=0,Y=0,Z=0),Scale3D=(X=1,Y=1,Z=1))),(Item=(Type=Bone,Name=\"off_thigh_chain_a_l\"),Transform=(Rotation=(X=0,Y=0,Z=0,W=1),Translation=(X=0,Y=0,Z=0),Scale3D=(X=1,Y=1,Z=1))),(Ite ...(truncated)"
    WeightMinimum=0
    ItemToModify.1.Transform.Rotation=@RigVMFunction_MathQuaternionFromEuler.Result
    ItemToModify.2.Transform.Rotation=&RigVMFunction_MathQuaternionFromEuler(Euler=(X=15,Y=-5,Z=20)).Result
    ExecutePin=@RigVMFunction_Sequence.A
    Weight=@n2.Value
   )
   > &RigUnit_ModifyTransforms(
    ItemToModify="((Item=(Type=Bone,Name=\"None\"),Transform=(Rotation=(X=0,Y=0,Z=0,W=1),Translation=(X=0,Y=0,Z=0),Scale3D=(X=1,Y=1,Z=1))),(Item=(Type=Bone,Name=\"base_Jacket_front_l\"),Transform=(Rotation=(X=0,Y=0.1736,Z=0,W=0.9848),Translation=(X=0,Y=0,Z=0),Scale3D=(X=1,Y=1,Z ...(truncated)"
    WeightMinimum=0
    Weight=@n2.Value
   )
   RigVMFunctionReturnNode(ExecuteContext=@RigVMFunction_Sequence.B)
  ]}
 New Function_ContainedGraph {nodes [
   n1 = RigVMFunctionEntryNode()
   RigVMFunction_Sequence = &RigVMFunction_Sequence(ExecuteContext=@n1.ExecuteContext)
   n2 = RigVMVariableNode() {variable=Pub_IsKneeSliding}
   &RigUnit_ModifyTransforms(
    ItemToModify="((Item=(Type=Bone,Name=\"None\"),Transform=(Rotation=(X=0,Y=0,Z=0,W=1),Translation=(X=0,Y=0,Z=0),Scale3D=(X=1,Y=1,Z=1))),(Item=(Type=Bone,Name=\"off_thigh_chain_a_l\"),Transform=(Rotation=(X=0.0404,Y=0.0811,Z=-0.0977,W=0.9911),Translation=(X=0.5013,Y=0.3067,Z= ...(truncated)"
    WeightMinimum=0
    ExecutePin=@RigVMFunction_Sequence.A
    Weight=@n2.Value
   )
   > &RigUnit_ModifyTransforms(
    ItemToModify="((Item=(Type=Bone,Name=\"None\"),Transform=(Rotation=(X=0,Y=0,Z=0,W=1),Translation=(X=0,Y=0,Z=0),Scale3D=(X=1,Y=1,Z=1))),(Item=(Type=Bone,Name=\"off_thigh_chain_a_r\"),Transform=(Rotation=(X=0.1577,Y=-0.3528,Z=-0.3764,W=0.842),Translation=(X=0,Y=0,Z=0),Scale3D ...(truncated)"
    WeightMinimum=0
    Weight=@n2.Value
   )
   > &RigUnit_ModifyTransforms(
    ItemToModify="((Item=(Type=Bone,Name=\"None\"),Transform=(Rotation=(X=0,Y=0,Z=0,W=1),Translation=(X=0,Y=0,Z=0),Scale3D=(X=1,Y=1,Z=1))),(Item=(Type=Bone,Name=\"base_Jacket_front_l\"),Transform=(Rotation=(X=-0.1294,Y=-0.483,Z=-0.2241,W=0.8365),Translation=(X=0,Y=0,Z=0),Scale3 ...(truncated)"
    WeightMinimum=0
    Weight=@n2.Value
   )
   n3 = RigVMVariableNode() {variable=Pub_IsBaseballSliding}
   > &RigUnit_ModifyTransforms(
    ItemToModify="((Item=(Type=Bone,Name=\"None\"),Transform=(Rotation=(X=0,Y=0,Z=0,W=1),Translation=(X=0,Y=0,Z=0),Scale3D=(X=1,Y=1,Z=1))),(Item=(Type=Bone,Name=\"off_thigh_chain_a_l\"),Transform=(Rotation=(X=0.0404,Y=0.0811,Z=-0.0977,W=0.9911),Translation=(X=0.3195,Y=-0.851,Z= ...(truncated)"
    WeightMinimum=0
    Weight=@n3.Value
   )
   > &RigUnit_ModifyTransforms(
    ItemToModify="((Item=(Type=Bone,Name=\"base_Jacket_front_l\"),Transform=(Rotation=(X=0.0686,Y=-0.7615,Z=-0.1085,W=0.6353),Translation=(X=0,Y=0,Z=0),Scale3D=(X=0.8,Y=1,Z=1))),(Item=(Type=Bone,Name=\"base_Jacket_a_l\"),Transform=(Rotation=(X=-0.1213,Y=-0.6337,Z=-0.1681,W=0.74 ...(truncated)"
    WeightMinimum=0
    Weight=@n3.Value
   )
   > DeltaRotation v4 = RigVMFunctionReferenceNode(
    Rotation Order=ZYX
    child=(Type=Bone,Name="spine_05")
    debugRotation=false
    outputMinMax=(X=0,Y=1)
    parent=(Type=Null,Name="spine_05_null")
    premult=1
    rotationMinMax=(X=0,Y=30)
    triggerAxis=Z
    twistAxis=(X=1,Y=0,Z=0)
   )
   > &RigUnit_ModifyTransforms(
    ItemToModify="((Item=(Type=Bone,Name=\"dyn_necklace_1\"),Transform=(Rotation=(X=0,Y=0,Z=0.1736,W=0.9848),Translation=(X=0,Y=0,Z=0),Scale3D=(X=1,Y=1,Z=1))),(Item=(Type=Bone,Name=\"dyn_necklace_2\"),Transform=(Rotation=(X=0,Y=0,Z=0.0872,W=0.9962),Translation=(X=0,Y=0,Z=0),Sca ...(truncated)"
    WeightMinimum=0
    Weight=@"DeltaRotation v4".outputWeight
   )
   > DeltaRotation v4_1 = RigVMFunctionReferenceNode(
    Rotation Order=ZYX
    child=(Type=Bone,Name="spine_05")
    debugRotation=false
    outputMinMax=(X=0,Y=1)
    parent=(Type=Null,Name="spine_05_null")
    premult=1
    rotationMinMax=(X=0,Y=30)
    triggerAxis=Y
    twistAxis=(X=1,Y=0,Z=0)
   )
   > &RigUnit_ModifyTransforms(
    ItemToModify="((Item=(Type=Bone,Name=\"dyn_necklace_1\"),Transform=(Rotation=(X=0,Y=0,Z=0.0436,W=0.999),Translation=(X=0,Y=0,Z=0),Scale3D=(X=1,Y=1,Z=1))),(Item=(Type=Bone,Name=\"dyn_necklace_2\"),Transform=(Rotation=(X=0,Y=0,Z=0,W=1),Translation=(X=0,Y=0,Z=0),Scale3D=(X=1,Y ...(truncated)"
    WeightMinimum=0
    Weight=@"DeltaRotation v4_1".outputWeight
   )
   > DeltaRotation v4_2 = RigVMFunctionReferenceNode(
    Rotation Order=ZYX
    child=(Type=Bone,Name="spine_05")
    debugRotation=false
    outputMinMax=(X=0,Y=1)
    parent=(Type=Null,Name="spine_05_null")
    premult=1
    rotationMinMax=(X=0,Y=-30)
    triggerAxis=Y
    twistAxis=(X=1,Y=0,Z=0)
   )
   > &RigUnit_ModifyTransforms(
    ItemToModify="((Item=(Type=Bone,Name=\"dyn_necklace_1\"),Transform=(Rotation=(X=0,Y=0,Z=0.0436,W=0.999),Translation=(X=0,Y=0,Z=0),Scale3D=(X=1,Y=1,Z=1))),(Item=(Type=Bone,Name=\"dyn_necklace_2\"),Transform=(Rotation=(X=0,Y=0,Z=0,W=1),Translation=(X=0,Y=0,Z=0),Scale3D=(X=1,Y ...(truncated)"
    WeightMinimum=0
    Weight=@"DeltaRotation v4_2".outputWeight
   )
   RigVMFunctionReturnNode(ExecuteContext=@RigVMFunction_Sequence.B)
  ]}
 New Function_ContainedGraph {nodes [
   n1 = RigVMFunctionEntryNode()
   RigVMFunction_MathBoolToInteger = &RigVMFunction_MathBoolToInteger(Value=@n1.onlyTwist)
   RigVMFunction_MathQuaternionSwingTwist = &RigVMFunction_MathQuaternionSwingTwist(
    TwistAxis=@n1.twistAxis
    Input=&RigUnit_GetRelativeTransformForItem(bChildInitial=False, bParentInitial=False, Child=@n1.child, Parent=@n1.parent).RelativeTransform.Rotation
   )
   n2 = RigVMDispatchNode(
    Values=((),())
    Index=@RigVMFunction_MathBoolToInteger.Result
    Values.0=@RigVMFunction_MathQuaternionSwingTwist.Swing
    Values.1=@RigVMFunction_MathQuaternionSwingTwist.Twist
   ) {template="DISPATCH_RigVMDispatch_SelectInt32(in Index,in Values,out Result)"}
   RigVMFunction_MathQuaternionToEuler = &RigVMFunction_MathQuaternionToEuler(Value=@n2.Result, RotationOrder=@n1.Rotation Order)
   > DISPATCH_RigVMDispatch_Print = RigVMDispatchNode(ScreenColor=(R=1,G=1,B=1,A=1), ScreenDuration=1.5, Enabled=@n1.debugRotation, Value=@RigVMFunction_MathQuaternionToEuler.Result) {template="DISPATCH_RigVMDispatch_Print(in Prefix,in Value,in Enabled,in ScreenDuration,in ScreenColor)"}
   Cast = RigVMDispatchNode(Value=@n1.triggerAxis) {template="DISPATCH_RigVMDispatch_CastEnumToInt(in Value,out Result)"}
   DISPATCH_RigVMDispatch_SelectInt32 = RigVMDispatchNode(
    Values=(0,0,0)
    Index=@Cast.Result
    Values.0=@RigVMFunction_MathQuaternionToEuler.Result.X
    Values.1=@RigVMFunction_MathQuaternionToEuler.Result.Y
    Values.2=@RigVMFunction_MathQuaternionToEuler.Result.Z
   ) {template="DISPATCH_RigVMDispatch_SelectInt32(in Index,in Values,out Result)"}
   Remap = &RigVMFunction_MathDoubleRemap(
    bClamp=False
    Value=@DISPATCH_RigVMDispatch_SelectInt32.Result
    TargetMinimum=@n1.outputMinMax.X
    TargetMaximum=@n1.outputMinMax.Y
    SourceMinimum=@n1.rotationMinMax.X
    SourceMaximum=@n1.rotationMinMax.Y
   )
   Absolute = &RigVMFunction_MathDoubleAbs(Value=@Remap.Result)
   Multiply = &RigVMFunction_MathDoubleMul(
    A=&RigVMFunction_MathDoubleClamp(Value=RigVMDispatchNode(True=@Absolute.Result, Condition=@n1.absoluteValue, False=@Remap.Result) {template="DISPATCH_RigVMDispatch_If(in Condition,in True,in False,out Result)"}.Result).Result
    B=@n1.premult
   )
   > RigVMFunctionReturnNode(outputQuat=@n2.Result, outputWeight=@Multiply.Result)
  ]}
 RigVMFunctionLibrary {nodes [
   Constraints = RigVMCollapseNode(ExecuteContext=())
   Constrollers = RigVMCollapseNode(Debug Controllers=False, ExecuteContext=())
   Crouching_def = RigVMCollapseNode(ExecuteContext=())
   DeltaRotation = RigVMCollapseNode(ExecuteContext=())
   DeltaRotation v4 = RigVMCollapseNode(
    ExecuteContext=()
    Rotation Order=ZYX
    child=(Type=Bone,Name="None")
    outputMinMax=(X=0,Y=1)
    parent=(Type=Bone,Name="None")
    premult=1
    rotationMinMax=(X=0,Y=0)
    twistAxis=(X=1,Y=0,Z=0)
   )
   KneeSliding_def = RigVMCollapseNode(ExecuteContext=())
   Leg_Chains = RigVMCollapseNode(ExecuteContext=())
   Neck_fix = RigVMCollapseNode(ExecuteContext=())
   Variables = RigVMCollapseNode(ExecuteContext=())
  ]}
 RigVMModel {nodes [
   RigUnit_BeginExecution = &RigUnit_BeginExecution()
   Constrollers = RigVMFunctionReferenceNode(Debug Controllers=false, ExecuteContext=@RigUnit_BeginExecution.ExecutePin)
   > FN_DeformRig_FNC = RigVMFunctionReferenceNode(
    DeformRig_ Curve="(deformRig_l_upperarm_up_alpha_72_#1=true,deformRig_r_upperarm_up_alpha_73_#2=true,deformRig_l_upperarm_fr_alpha_74_#3=true,deformRig_r_upperarm_fr_alpha_75_CB51B80C4 ...(truncated)"
    DeformRig_Bool="(upperarm_1_#4=true,elbow_3_#5=true,wrist_7_#6=True,groin_8_#7=true,glute_17_#8=true,knee_9_40C408324CA261 ...(truncated)"
    DeformRig_Debug=(upperarm_debug_18_#4=False,wrist_debug_20_#6=False,lat_debug_26_#9=False,groinGlute_debug_27_#7=False)
    DeformRig_Value="(upperarm_up_18_#10=(Rotation=(X=0,Y=0,Z=0,W=1),Translation=(X=0,Y=0,Z=0.8),Scale3D=(X=0.7,Y=1,Z=1)),upperarm_fwd_fr_19_#11=(Rotation=(X=0,Y=0,Z=0,W=1),Translation=(X=0,Y=10,Z=0),Scale3D=(X=0.5,Y=1, ...(truncated)"
   )
   > Variables = RigVMFunctionReferenceNode()
   > Constraints = RigVMFunctionReferenceNode()
   > Leg_Chains = RigVMFunctionReferenceNode()
   > Crouching_def = RigVMFunctionReferenceNode()
   > KneeSliding_def = RigVMFunctionReferenceNode()
   > Neck_fix = RigVMFunctionReferenceNode()
   &RigUnit_ParentConstraint(
    AdvancedSettings=(RotationOrderForFilter=XZY)
    Child=(Type=Bone,Name="base_neck_c")
    ExecutePin=()
    Parents="((Item=(Type=Bone,Name=\"FACIAL_C_Jaw\"),Weight=0.0056),(Item=(Type=Bone,Name=\"FACIAL_R_UnderChin\"),Weight=0.0122),(Item=(Type=Bone,Name=\"spine_05\"),Weight=0.018),(Item=(Type=Bone,Name=\"head\"),Weight=0.0989),(Item=(Type=Bone,Name=\"neck_01\"),Weight=0.1523),(I ...(truncated)"
    Weight=1
   )
   &RigUnit_ParentConstraint(
    AdvancedSettings=(RotationOrderForFilter=XZY)
    Child=(Type=Bone,Name="base_neck_c")
    ExecutePin=()
    Parents="((Item=(Type=Bone,Name=\"FACIAL_C_Jaw\"),Weight=0.0056),(Item=(Type=Bone,Name=\"FACIAL_R_UnderChin\"),Weight=0.0122),(Item=(Type=Bone,Name=\"spine_05\"),Weight=0.018),(Item=(Type=Bone,Name=\"head\"),Weight=0.0989),(Item=(Type=Bone,Name=\"neck_01\"),Weight=0.1523),(I ...(truncated)"
    Weight=1
   )
   &RigUnit_ParentConstraint(
    AdvancedSettings=(RotationOrderForFilter=XZY)
    Child=(Type=Bone,Name="base_neck_c")
    ExecutePin=()
    Parents="((Item=(Type=Bone,Name=\"FACIAL_C_Jaw\"),Weight=0.0056),(Item=(Type=Bone,Name=\"FACIAL_R_UnderChin\"),Weight=0.0122),(Item=(Type=Bone,Name=\"spine_05\"),Weight=0.018),(Item=(Type=Bone,Name=\"head\"),Weight=0.0989),(Item=(Type=Bone,Name=\"neck_01\"),Weight=0.1523),(I ...(truncated)"
    Weight=1
   )
   &RigUnit_ParentConstraint(
    AdvancedSettings=(RotationOrderForFilter=XZY)
    Child=(Type=Bone,Name="base_neck_c")
    ExecutePin=()
    Parents="((Item=(Type=Bone,Name=\"FACIAL_C_Jaw\"),Weight=0.0056),(Item=(Type=Bone,Name=\"FACIAL_R_UnderChin\"),Weight=0.0122),(Item=(Type=Bone,Name=\"spine_05\"),Weight=0.018),(Item=(Type=Bone,Name=\"head\"),Weight=0.0989),(Item=(Type=Bone,Name=\"neck_01\"),Weight=0.1523),(I ...(truncated)"
    Weight=1
   )
  ]}
]
hierarchy [
 root {loc=<0 0 0>, rot=<0 0 0 1>}
 pelvis {parent=root, loc=<0 -0.7071 93.1191>, rot=<0.0127 0.707 -0.0127 -0.707>}
 spine_01 {parent=pelvis, loc=<6.7494 -0.6388 0>, rot=<0 0 0.034 -0.9994>}
 spine_02 {parent=spine_01, loc=<7.252 -0.035 0>, rot=<0 0 0.0238 -0.9997>}
 spine_03 {parent=spine_02, loc=<6.0106 0 0>, rot=<0 0 0.0149 -0.9999>}
 spine_04 {parent=spine_03, loc=<8.1023 0 0>, rot=<0 0 -0.0132 -0.9999>}
 spine_05 {parent=spine_04, loc=<10.0142 0 0>, rot=<0 0 -0.0015 -1>}
 attach_backpack {parent=spine_05, loc=<-0.123 -13.473 0>, rot=<-0.0093 -0.707 0.0093 -0.707>}
 neck_01 {parent=spine_05, loc=<11.3138 0.1529 0>, rot=<0 0 -0.2003 -0.9797>}
 neck_02 {parent=neck_01, loc=<4.5627 0 0>, rot=<0 0 0.0504 -0.9987>}
 head {parent=neck_02, loc=<5.3914 0 0>, rot=<0 0 0.115 -0.9934>}
 base_earing_l {parent=head, loc=<1.0667 1.2043 -6.6592>, rot=<0.1235 -0.9406 -0.3162 0.01>}
 dyn_earing_l {parent=base_earing_l, loc=<0 0 0>, rot=<0 0 0 1>}
 base_earing_r {parent=head, loc=<1.0669 1.2043 6.6592>, rot=<0.1235 -0.9406 -0.3162 0.01>}
 dyn_earing_r {parent=base_earing_r, loc=<0 0 0>, rot=<0 0 0 1>}
 base_hair_d_01_l {parent=head, loc=<8.6657 0.9446 -6.9141>, rot=<-0.4402 -0.8978 0.0074 0.0112>}
 dyn_hair_d_01_l {parent=base_hair_d_01_l, loc=<0 0 0>, rot=<0 0 0 1>}
 base_hair_e_01_l {parent=head, loc=<7.5669 0.2947 -6.8514>, rot=<-0.2981 -0.9533 0.0388 0.0302>}
 dyn_hair_e_01_l {parent=base_hair_e_01_l, loc=<0 0 0>, rot=<0 0 0 1>}
 base_hair_f_01_l {parent=head, loc=<14.8522 7.846 -1.9234>, rot=<-0.3595 -0.556 -0.3507 -0.6623>}
 dyn_hair_f_01_l {parent=base_hair_f_01_l, loc=<0 0 0>, rot=<0 0 0 1>}
 base_hair_g_01_l {parent=head, loc=<16.9348 3.3088 -1.5562>, rot=<-0.5787 -0.3537 -0.356 -0.6429>}
 dyn_hair_g_01_l {parent=base_hair_g_01_l, loc=<0 0 0>, rot=<0 0 0 1>}
 base_hair_h_01_l {parent=head, loc=<1.3663 -3.7411 -4.3784>, rot=<0.3638 -0.7309 0.4886 -0.3078>}
 dyn_hair_h_01_l {parent=base_hair_h_01_l, loc=<0 0 0>, rot=<0 0 0 1>}
 base_hair_i_01_l {parent=head, loc=<1.1475 -5.2063 -2.4322>, rot=<0.4412 -0.6966 0.5046 -0.2558>}
 dyn_hair_i_01_l {parent=base_hair_i_01_l, loc=<0 0 0>, rot=<0 0 0 1>}
 base_hair_j_01_l {parent=head, loc=<2.6592 -5.9227 -1.7746>, rot=<0.3691 -0.5857 0.696 -0.1903>}
 dyn_hair_j_01_l {parent=base_hair_j_01_l, loc=<0 0 0>, rot=<0 0 0 1>}
 base_hair_k_01_l {parent=head, loc=<3.5268 -6.4166 -0.4654>, rot=<0.3691 -0.5857 0.696 -0.1903>}
 dyn_hair_k_01_l {parent=base_hair_k_01_l, loc=<0 0 0>, rot=<0 0 0 1>}
 base_hair_l_01_l {parent=head, loc=<0.7202 -6.4661 0.0064>, rot=<0.5131 -0.4804 0.623 -0.3433>}
 dyn_hair_l_01_l {parent=base_hair_l_01_l, loc=<0 0 0>, rot=<0 0 0 1>}
 base_hair_m_01_l {parent=head, loc=<7.7055 5.6757 -6.1065>, rot=<0.2522 -0.9624 0.0457 -0.0902>}
 dyn_hair_m_01_l {parent=base_hair_m_01_l, loc=<0 0 0>, rot=<0 0 0 1>}
 base_hair_n_01_l {parent=head, loc=<9.0088 3.8399 -6.4389>, rot=<0.0875 -0.9921 -0.0885 -0.0179>}
 dyn_hair_n_01_l {parent=base_hair_n_01_l, loc=<0 0 0>, rot=<0 0 0 1>}
 base_hair_a_01_r {parent=head, loc=<8.4876 9.3581 2.6133>, rot=<0.0353 0.6511 0.7389 -0.1697>}
 dyn_hair_a_01_r {parent=base_hair_a_01_r, loc=<0 0 0>, rot=<0 0 0 1>}
 dyn_hair_a_02_r {parent=dyn_hair_a_01_r, loc=<1.4433 0.1001 0.1005>, rot=<0.015 0.0074 -0.0745 -0.9971>}
 base_hair_b_01_r {parent=head, loc=<10.4398 6.1285 5.4914>, rot=<-0.017 0.3867 0.9136 0.1244>}
 dyn_hair_b_01_r {parent=base_hair_b_01_r, loc=<0 0 0>, rot=<0 0 0 1>}
 dyn_hair_b_02_r {parent=dyn_hair_b_01_r, loc=<1.3851 -0.0045 -0.0503>, rot=<0.0711 0.0052 -0.059 -0.9957>}
 base_hair_c_01_r {parent=head, loc=<9.4538 4.9753 6.1823>, rot=<0.0198 0.4296 0.8957 0.1135>}
 dyn_hair_c_01_r {parent=base_hair_c_01_r, loc=<0 0 0>, rot=<0 0 0 1>}
 dyn_hair_c_02_r {parent=dyn_hair_c_01_r, loc=<2.1109 0.0276 -0.0007>, rot=<-0.053 0.0417 -0.2548 -0.9646>}
 base_hair_d_01_r {parent=head, loc=<10.5765 2.0607 6.147>, rot=<0.007 0.0105 0.9514 0.3078>}
 dyn_hair_d_01_r {parent=base_hair_d_01_r, loc=<0 0 0>, rot=<0 0 0 1>}
 base_hair_e_01_r {parent=head, loc=<0.9565 -6.526 1.6232>, rot=<0.4897 -0.3961 0.7292 -0.2673>}
 dyn_hair_e_01_r {parent=base_hair_e_01_r, loc=<0 0 0>, rot=<0 0 0 1>}
 base_hair_f_01_r {parent=head, loc=<2.7429 -5.8426 2.2916>, rot=<0.5093 -0.3497 0.7569 -0.2131>}
 dyn_hair_f_01_r {parent=base_hair_f_01_r, loc=<0 0 0>, rot=<0 0 0 1>}
 base_hair_g_01_r {parent=head, loc=<1.444 -4.6744 3.0744>, rot=<0.5401 -0.301 0.7717 -0.1488>}
 dyn_hair_g_01_r {parent=base_hair_g_01_r, loc=<0 0 0>, rot=<0 0 0 1>}
 base_hair_h_01_r {parent=head, loc=<1.7954 -4.1405 4.1411>, rot=<0.322 -0.3349 0.8826 -0.0715>}
 dyn_hair_h_01_r {parent=base_hair_h_01_r, loc=<0 0 0>, rot=<0 0 0 1>}
 base_hair_i_01_r {parent=head, loc=<1.6784 -3.5398 4.9744>, rot=<0.5142 -0.3009 0.803 -0.0153>}
 dyn_hair_i_01_r {parent=base_hair_i_01_r, loc=<0 0 0>, rot=<0 0 0 1>}
 base_hair_j_01_r {parent=head, loc=<2.4179 -2.1277 5.7893>, rot=<-0.1604 -0.225 0.9139 0.2974>}
 dyn_hair_j_01_r {parent=base_hair_j_01_r, loc=<0 0 0>, rot=<0 0 0 1>}
 base_hair_k_01_r {parent=head, loc=<2.6218 -1.2924 5.7124>, rot=<-0.0451 0.19 -0.763 -0.6162>}
 dyn_hair_k_01_r {parent=base_hair_k_01_r, loc=<0 0 0>, rot=<0 0 0 1>}
 base_hair_l_01_r {parent=head, loc=<7.2694 -1.2139 6.4939>, rot=<-0.0446 0.1643 -0.7622 -0.6246>}
 dyn_hair_l_01_r {parent=base_hair_l_01_r, loc=<0 0 0>, rot=<0 0 0 1>}
 base_hair_m_01_r {parent=head, loc=<3.7478 -4.904 3.8764>, rot=<0.3589 -0.458 0.785 -0.2127>}
 dyn_hair_m_01_r {parent=base_hair_m_01_r, loc=<0 0 0>, rot=<0 0 0 1>}
 base_hair_n_01_r {parent=head, loc=<6.6005 -6.1865 3.7883>, rot=<0.1409 -0.5498 0.7993 -0.1973>}
 dyn_hair_n_01_r {parent=base_hair_n_01_r, loc=<0 0 0>, rot=<0 0 0 1>}
 base_hair_o_01_r {parent=head, loc=<5.6474 -5.2736 4.7579>, rot=<0.0769 -0.4545 0.8874 0.005>}
 dyn_hair_o_01_r {parent=base_hair_o_01_r, loc=<0 0 0>, rot=<0 0 0 1>}
 base_main_hair_a_l {parent=head, loc=<11.04 9.7584 -0.8122>, rot=<-0.0332 -0.7522 -0.6564 -0.047>}
 dyn_main_hair_a_l {parent=base_main_hair_a_l, loc=<0 0 0>, rot=<0 0 0 1>}
 base_hair_a_01_l {parent=dyn_main_hair_a_l, loc=<2.4622 -0.3532 0.3715>, rot=<0.1338 -0.089 -0.1805 -0.9704>}
 dyn_hair_a_01_l {parent=base_hair_a_01_l, loc=<0 0 0>, rot=<0 0 0 1>}
 base_main_hair_b_l {parent=head, loc=<8.7037 7.206 -5.432>, rot=<-0.0167 -0.9603 -0.2759 -0.038>}
 dyn_main_hair_b_l {parent=base_main_hair_b_l, loc=<0 0 0>, rot=<0 0 0 1>}
 base_hair_c_01_l {parent=dyn_main_hair_b_l, loc=<2.5915 -0.5664 0.0947>, rot=<-0.1641 -0.1318 -0.1497 -0.9661>}
 dyn_hair_c_01_l {parent=base_hair_c_01_l, loc=<0 0 0>, rot=<0 0 0 1>}
 base_main_hair_c_l {parent=head, loc=<10.2536 8.8625 -3.7739>, rot=<-0.0032 -0.8153 -0.5775 -0.0414>}
 dyn_main_hair_c_l {parent=base_main_hair_c_l, loc=<0 0 0>, rot=<0 0 0 1>}
 base_hair_b_01_l {parent=dyn_main_hair_c_l, loc=<4.008 -1.7057 -0.2277>, rot=<-0.2467 0.4349 0.0879 -0.8615>}
 dyn_hair_b_01_l {parent=base_hair_b_01_l, loc=<0 0 0>, rot=<0 0 0 1>}
 FACIAL_C_FacialRoot {parent=head, loc=<2.7225 1.07 0>, rot=<0.5021 -0.4979 -0.5021 -0.4979>}
 FACIAL_L_ForeheadIn {parent=FACIAL_C_FacialRoot, loc=<1.4246 -4.1395 7.0615>, rot=<-0.0609 -0.0696 0.0043 -0.9957>}
 FACIAL_R_ForeheadIn {parent=FACIAL_C_FacialRoot, loc=<-1.4246 -4.1395 7.0615>, rot=<-0.0609 0.0696 -0.0043 -0.9957>}
 FACIAL_L_EyeCornerInner {parent=FACIAL_C_FacialRoot, loc=<1.875 -1.5304 6.8058>, rot=<0 0 0 -1>}
 FACIAL_R_EyeCornerInner {parent=FACIAL_C_FacialRoot, loc=<-1.875 -1.5304 6.8058>, rot=<0 0 0 -1>}
 FACIAL_L_EyeCornerOuter {parent=FACIAL_C_FacialRoot, loc=<4.0909 -1.6071 6.6562>, rot=<0 -0.3827 0 -0.9239>}
 FACIAL_R_EyeCornerOuter {parent=FACIAL_C_FacialRoot, loc=<-4.0909 -1.6071 6.6562>, rot=<0 0.3827 0 -0.9239>}
 FACIAL_L_EyesackLower {parent=FACIAL_C_FacialRoot, loc=<3.4067 -0.1619 6.4872>, rot=<-0.0223 -0.1398 -0.0223 -0.9897>}
 FACIAL_R_EyesackLower {parent=FACIAL_C_FacialRoot, loc=<-3.4067 -0.1619 6.4872>, rot=<-0.0223 0.1398 0.0223 -0.9897>}
 FACIAL_L_CheekInner {parent=FACIAL_C_FacialRoot, loc=<1.879 0.1459 6.6741>, rot=<-0.1233 -0.088 -0.1972 -0.9686>}
 FACIAL_R_CheekInner {parent=FACIAL_C_FacialRoot, loc=<-1.879 0.1459 6.6741>, rot=<-0.1233 0.088 0.1972 -0.9686>}
 FACIAL_L_CheekOuter {parent=FACIAL_C_FacialRoot, loc=<3.6003 0.2795 5.4607>, rot=<0.0551 -0.377 0.1003 -0.9191>}
 FACIAL_R_CheekOuter {parent=FACIAL_C_FacialRoot, loc=<-3.6003 0.2795 5.4607>, rot=<0.0551 0.377 -0.1003 -0.9191>}
 FACIAL_L_NasolabialBulge {parent=FACIAL_C_FacialRoot, loc=<3.5502 3.5741 5.7721>, rot=<0.0579 -0.3774 0.0238 -0.9239>}
 FACIAL_R_NasolabialBulge {parent=FACIAL_C_FacialRoot, loc=<-3.5502 3.5741 5.7721>, rot=<0.0579 0.3774 -0.0238 -0.9239>}
 FACIAL_L_NasolabialFurrow {parent=FACIAL_C_FacialRoot, loc=<2.7559 2.069 7.588>, rot=<-0.0333 -0.1258 -0.3638 -0.9223>}
 FACIAL_R_NasolabialFurrow {parent=FACIAL_C_FacialRoot, loc=<-2.7559 2.069 7.588>, rot=<-0.0333 0.1258 0.3638 -0.9223>}
 FACIAL_L_CheekLower {parent=FACIAL_C_FacialRoot, loc=<4.1249 3.3751 4.8564>, rot=<0.0345 -0.427 0.0577 -0.9018>}
 FACIAL_R_CheekLower {parent=FACIAL_C_FacialRoot, loc=<-4.1249 3.3751 4.8564>, rot=<0.0345 0.427 -0.0577 -0.9018>}
 FACIAL_C_Nose {parent=FACIAL_C_FacialRoot, loc=<0 -0.3249 8.1773>, rot=<-0.1305 0 0 -0.9914>}
 FACIAL_C_NoseLower {parent=FACIAL_C_Nose, loc=<0 2.8266 0.0566>, rot=<0.1305 0 0 -0.9914>}
 FACIAL_L_Nostril {parent=FACIAL_C_Nose, loc=<1.1703 2.3286 0.0348>, rot=<0.0173 0.1305 0.0023 -0.9913>}
 FACIAL_R_Nostril {parent=FACIAL_C_Nose, loc=<-1.1703 2.3286 0.0348>, rot=<0.0173 -0.1305 -0.0023 -0.9913>}
 FACIAL_C_NoseTip {parent=FACIAL_C_Nose, loc=<0 1.9367 0.9374>, rot=<0.0872 0 0 -0.9962>}
 FACIAL_C_MouthUpper {parent=FACIAL_C_FacialRoot, loc=<0 3.8936 6.5794>, rot=<0 0 0 -1>}
 FACIAL_C_LipUpper {parent=FACIAL_C_MouthUpper, loc=<0 -0.5779 2.023>, rot=<-0.0872 0 0 -0.9962>}
 FACIAL_L_LipUpper {parent=FACIAL_C_MouthUpper, loc=<1.2627 -0.3158 1.7478>, rot=<-0.0778 -0.1601 -0.0379 -0.9833>}
 FACIAL_R_LipUpper {parent=FACIAL_C_MouthUpper, loc=<-1.2627 -0.3158 1.7478>, rot=<-0.0778 0.1601 0.0379 -0.9833>}
 FACIAL_L_LipUpperOuter {parent=FACIAL_C_MouthUpper, loc=<2.0098 -0.0117 1.1617>, rot=<-0.0983 -0.2294 -0.0983 -0.9634>}
 FACIAL_R_LipUpperOuter {parent=FACIAL_C_MouthUpper, loc=<-2.0098 -0.0117 1.1617>, rot=<-0.0983 0.2294 0.0983 -0.9634>}
 FACIAL_L_LipCorner {parent=FACIAL_C_MouthUpper, loc=<2.4284 0.3428 0.6909>, rot=<0 -0.2079 0 -0.9781>}
 FACIAL_R_LipCorner {parent=FACIAL_C_MouthUpper, loc=<-2.4284 0.3428 0.6909>, rot=<0 0.2079 0 -0.9781>}
 FACIAL_L_Masseter {parent=FACIAL_C_FacialRoot, loc=<6.0167 1.1642 3.3148>, rot=<0.0693 0.1129 0.0079 -0.9912>}
 FACIAL_R_Masseter {parent=FACIAL_C_FacialRoot, loc=<-6.0167 1.1642 3.3148>, rot=<0.0693 -0.1129 -0.0079 -0.9912>}
 FACIAL_C_UnderChin {parent=FACIAL_C_FacialRoot, loc=<0 8.7669 3.9666>, rot=<0.5 0 0 -0.866>}
 FACIAL_L_UnderChin {parent=FACIAL_C_FacialRoot, loc=<3.3571 7.4821 2.6112>, rot=<0.4025 -0.355 0.0823 -0.8398>}
 FACIAL_R_UnderChin {parent=FACIAL_C_FacialRoot, loc=<-3.3569 7.482 2.612>, rot=<0.4025 0.355 -0.0823 -0.8398>}
 FACIAL_C_LowerLipRotation {parent=FACIAL_C_FacialRoot, loc=<0 1.9303 1.5182>, rot=<0.342 0 0 -0.9397>}
 FACIAL_C_MouthLower {parent=FACIAL_C_LowerLipRotation, loc=<0 -1.7493 5.1391>, rot=<-0.342 0 0 -0.9397>}
 FACIAL_C_LipLower {parent=FACIAL_C_MouthLower, loc=<0 1.4444 1.7154>, rot=<0.0872 0 0 -0.9962>}
 FACIAL_L_LipLower {parent=FACIAL_C_MouthLower, loc=<1.0049 1.2869 1.5487>, rot=<0.0927 -0.1295 0.0273 -0.9869>}
 FACIAL_R_LipLower {parent=FACIAL_C_MouthLower, loc=<-1.0049 1.2869 1.5487>, rot=<0.0927 0.1295 -0.0273 -0.9869>}
 FACIAL_L_LipLowerOuter {parent=FACIAL_C_MouthLower, loc=<1.8311 0.8462 1.1099>, rot=<0.0913 -0.2322 0.0318 -0.9679>}
 FACIAL_R_LipLowerOuter {parent=FACIAL_C_MouthLower, loc=<-1.8311 0.8462 1.1099>, rot=<0.0913 0.2322 -0.0318 -0.9679>}
 FACIAL_C_Jaw {parent=FACIAL_C_FacialRoot, loc=<0 1.9303 1.5182>, rot=<0.342 0 0 -0.9397>}
 FACIAL_C_Jawline {parent=FACIAL_C_Jaw, loc=<0 2.1535 7.8456>, rot=<0.2588 0 0 -0.9659>}
 FACIAL_L_ChinSide {parent=FACIAL_C_Jaw, loc=<2.9769 -0.2831 6.7982>, rot=<-0.2207 -0.3739 0.05 -0.8994>}
 FACIAL_R_ChinSide {parent=FACIAL_C_Jaw, loc=<-2.9769 -0.2831 6.7982>, rot=<-0.2207 0.3739 -0.05 -0.8994>}
 FACIAL_C_Chin {parent=FACIAL_C_Jaw, loc=<0 0.6215 7.5941>, rot=<-0.2588 0 0 -0.9659>}
 FACIAL_L_Ear {parent=FACIAL_C_FacialRoot, loc=<6.5283 -0.1723 0.5389>, rot=<-0.0262 -0.0007 -0.0262 -0.9993>}
 FACIAL_R_Ear {parent=FACIAL_C_FacialRoot, loc=<-6.5284 -0.1723 0.5389>, rot=<-0.0262 0.0007 0.0262 -0.9993>}
 FACIAL_L_Jawline {parent=FACIAL_C_FacialRoot, loc=<4.4861 5.1457 3.8606>, rot=<0.2092 0.1578 -0.0547 -0.9635>}
 FACIAL_R_Jawline {parent=FACIAL_C_FacialRoot, loc=<-4.4861 5.1457 3.8606>, rot=<0.2092 -0.1578 0.0547 -0.9635>}
 FACIAL_C_NoseBridge {parent=FACIAL_C_FacialRoot, loc=<0 -2.4419 8.3784>, rot=<0.0436 0 0 -0.999>}
 FACIAL_L_EyelidLowerB {parent=FACIAL_C_FacialRoot, loc=<2.8168 -1.7766 6.0495>, rot=<0.3173 0 0 -0.9483>}
 FACIAL_R_EyelidLowerB {parent=FACIAL_C_FacialRoot, loc=<-2.8168 -1.7766 6.0495>, rot=<0.3256 0 0 -0.9455>}
 FACIAL_C_Forehead {parent=FACIAL_C_FacialRoot, loc=<0 -4.0903 7.3344>, rot=<-0.0436 0 0 -0.999>}
 FACIAL_L_ForeheadMid {parent=FACIAL_C_FacialRoot, loc=<2.6324 -4.1539 6.6711>, rot=<-0.0712 -0.1757 -0.0206 -0.9816>}
 FACIAL_R_ForeheadMid {parent=FACIAL_C_FacialRoot, loc=<-2.6324 -4.1539 6.6711>, rot=<-0.0712 0.1757 0.0206 -0.9816>}
 FACIAL_L_ForeheadOut {parent=FACIAL_C_FacialRoot, loc=<3.7801 -3.8232 6.091>, rot=<-0.1215 -0.2557 -0.0792 -0.9558>}
 FACIAL_R_ForeheadOut {parent=FACIAL_C_FacialRoot, loc=<-3.7801 -3.8232 6.091>, rot=<-0.1215 0.2557 0.0792 -0.9558>}
 FACIAL_L_EyesackUpper {parent=FACIAL_C_FacialRoot, loc=<3.0221 -2.5328 7.677>, rot=<0.0487 -0.0843 -0.0646 -0.9932>}
 FACIAL_R_EyesackUpper {parent=FACIAL_C_FacialRoot, loc=<-3.0222 -2.5328 7.677>, rot=<0.0487 0.0843 0.0646 -0.9932>}
 FACIAL_L_EyelidUpperFurrow {parent=FACIAL_C_FacialRoot, loc=<2.9205 -2.4223 7.551>, rot=<-0.0301 -0.1509 -0.0797 -0.9849>}
 FACIAL_R_EyelidUpperFurrow {parent=FACIAL_C_FacialRoot, loc=<-2.9205 -2.4223 7.551>, rot=<-0.0301 0.1509 0.0797 -0.9849>}
 FACIAL_L_EyelidUpperA {parent=FACIAL_C_FacialRoot, loc=<2.8168 -1.7766 6.0495>, rot=<-0.1478 0 0 -0.989>}
 FACIAL_R_EyelidUpperA {parent=FACIAL_C_FacialRoot, loc=<-2.8168 -1.7766 6.0495>, rot=<-0.1478 0 0 -0.989>}
 FACIAL_L_EyelidLowerA {parent=FACIAL_C_FacialRoot, loc=<2.8168 -1.7766 6.0495>, rot=<0.284 0 0 -0.9588>}
 FACIAL_R_EyelidLowerA {parent=FACIAL_C_FacialRoot, loc=<-2.8168 -1.7766 6.0495>, rot=<0.284 0 0 -0.9588>}
 FACIAL_L_EyelidUpperB {parent=FACIAL_C_FacialRoot, loc=<2.8168 -1.7766 6.0495>, rot=<-0.225 0 0 -0.9744>}
 FACIAL_R_EyelidUpperB {parent=FACIAL_C_FacialRoot, loc=<-2.8168 -1.7766 6.0495>, rot=<-0.225 0 0 -0.9744>}
 base_neck_c {parent=neck_01, loc=<4.7683 6.6943 -0.0023>, rot=<0.1137 0.6979 0.6979 0.1137>}
 dyn_neck_c {parent=base_neck_c, loc=<0 0 0>, rot=<0 0 0 1>}
 clavicle_l {parent=spine_05, loc=<4.3575 -0.1466 -2.3223>, rot=<0.0465 -0.7289 -0.009 -0.6829>}
 upperarm_l {parent=clavicle_l, loc=<15.5111 0 0>, rot=<0.0226 -0.3234 -0.0088 -0.9459>}
 lowerarm_l {parent=upperarm_l, loc=<26.6514 0 0>, rot=<0.0391 -0.0099 -0.2453 -0.9686>}
 hand_l {parent=lowerarm_l, loc=<22.3491 0 0>, rot=<-0.7081 -0.0507 0.0975 -0.6975>}
 thumb_01_l {parent=hand_l, loc=<1.3118 -0.3191 -2.459>, rot=<0.6546 -0.4042 0.0664 -0.6354>}
 thumb_02_l {parent=thumb_01_l, loc=<5.9336 -0.0448 0.0401>, rot=<-0.0037 0.0213 0.298 -0.9543>}
 thumb_03_l {parent=thumb_02_l, loc=<2.2223 0.3682 -0.0907>, rot=<0.0118 -0.013 0.0066 -0.9998>}
 index_metacarpal_l {parent=hand_l, loc=<1.6535 0.0183 -2.96>, rot=<0.0158 -0.0124 -0.0173 -0.9996>}
 index_01_l {parent=index_metacarpal_l, loc=<8.3008 0.0357 -0.0524>, rot=<-0.0044 0.0485 0.0924 -0.9945>}
 index_02_l {parent=index_01_l, loc=<3.3587 -0.3004 0.0452>, rot=<-0.002 0.0027 0.3225 -0.9466>}
 index_03_l {parent=index_02_l, loc=<2.5578 -0.123 -0.0252>, rot=<-0.002 0.0067 0.2558 -0.9667>}
 middle_metacarpal_l {parent=hand_l, loc=<2.1856 -0.1725 -0.5984>, rot=<0.006 0.0168 -0.0329 -0.9993>}
 middle_01_l {parent=middle_metacarpal_l, loc=<7.3275 -0.2447 0.0129>, rot=<-0.0208 0.0042 0.0858 -0.9961>}
 middle_02_l {parent=middle_01_l, loc=<4.0123 0.1401 0.1175>, rot=<-0.0218 0.0258 0.3512 -0.9357>}
 middle_03_l {parent=middle_02_l, loc=<2.6077 0.0534 0.1373>, rot=<0.0082 0.0188 0.1708 -0.9851>}
 ring_metacarpal_l {parent=hand_l, loc=<2.2074 -0.3625 1.3638>, rot=<-0.0284 0.0722 -0.0458 -0.9959>}
 ring_01_l {parent=ring_metacarpal_l, loc=<7.0295 -0.6197 -0.1446>, rot=<-0.0489 -0.0252 0.1892 -0.9804>}
 ring_02_l {parent=ring_01_l, loc=<3.6174 0.4647 -0.1027>, rot=<0 0 0.2806 -0.9598>}
 ring_03_l {parent=ring_02_l, loc=<2.4152 0.114 0.0199>, rot=<0 0 0.1189 -0.9929>}
 pinky_metacarpal_l {parent=hand_l, loc=<1.7616 -0.9753 3.3814>, rot=<-0.0397 0.0814 -0.069 -0.9935>}
 pinky_01_l {parent=pinky_metacarpal_l, loc=<7.1709 -0.4831 -0.1094>, rot=<-0.0654 -0.0035 0.2242 -0.9723>}
 pinky_02_l {parent=pinky_01_l, loc=<2.8274 0.0893 -0.1043>, rot=<0.0037 0.0011 0.2854 -0.9584>}
 pinky_03_l {parent=pinky_02_l, loc=<2.2046 0.0307 -0.0078>, rot=<-0.0002 0.0098 0.1445 -0.9895>}
 hand_attach_l {parent=hand_l, loc=<8.5 -3 0>, rot=<0.7071 0.7071 0 0>}
 weapon_l {parent=hand_l, loc=<-0.1974 -3.4111 0.3811>, rot=<0 0 0 1>}
 deform_wrist_dn_base_l {parent=hand_l, loc=<0 -2.2342 0>, rot=<0 0 0 -1>}
 deform_wrist_dn_l {parent=deform_wrist_dn_base_l, loc=<0 0 0>, rot=<0 0 0 1>}
 deform_wrist_up_base_l {parent=hand_l, loc=<0 2.2682 0>, rot=<0 0 0 -1>}
 deform_wrist_up_l {parent=deform_wrist_up_base_l, loc=<0 0 0>, rot=<0 0 0 1>}
 lowerarm_twist_02_l {parent=lowerarm_l, loc=<9.6199 0 0>, rot=<0 0 0 -1>}
 lowerarm_twist_01_l {parent=lowerarm_l, loc=<17.8549 0 0>, rot=<0 0 0 -1>}
 deform_elbow_in_base_l {parent=lowerarm_l, loc=<0.2112 5.259 0.8274>, rot=<0 0 0 -1>}
 deform_elbow_in_l {parent=deform_elbow_in_base_l, loc=<0 0 0>, rot=<0 0 0 1>}
 deform_elbow_out_base_l {parent=lowerarm_l, loc=<0 -4 0>, rot=<0 0 0 -1>}
 deform_elbow_out_l {parent=deform_elbow_out_base_l, loc=<0 0 0>, rot=<0 0 0 1>}
 upperarm_twist_01_l {parent=upperarm_l, loc=<0.0041 0 0>, rot=<0 0 0 -1>}
 deform_upperarm_bk_base_l {parent=upperarm_twist_01_l, loc=<1.0813 -7.2193 -1.9453>, rot=<0 0 0 1>}
 deform_upperarm_bk_l {parent=deform_upperarm_bk_base_l, loc=<0 0 0>, rot=<0 0 0 1>}
 deform_upperarm_fr_base_l {parent=upperarm_twist_01_l, loc=<-0.3691 7.0506 -2.765>, rot=<0 0 0 1>}
 deform_upperarm_fr_l {parent=deform_upperarm_fr_base_l, loc=<0 0 0>, rot=<0 0 0 1>}
 deform_upperarm_up_base_l {parent=upperarm_twist_01_l, loc=<-3.2216 -1.1391 3.3311>, rot=<0 0 0 1>}
 deform_upperarm_up_l {parent=deform_upperarm_up_base_l, loc=<0 0 0>, rot=<0 0 0 1>}
 upperarm_twist_02_l {parent=upperarm_l, loc=<13.7941 0 0>, rot=<0 0 0 -1>}
 attach_cape {parent=spine_05, loc=<9.017 -11.714 0>, rot=<0.0616 -0.7044 -0.0616 -0.7044>}
 clavicle_r {parent=spine_05, loc=<4.3575 -0.1466 2.3223>, rot=<0.7289 0.0465 -0.6829 0.009>}
 upperarm_r {parent=clavicle_r, loc=<-15.5111 0 0>, rot=<0.0226 -0.3234 -0.0088 -0.9459>}
 lowerarm_r {parent=upperarm_r, loc=<-26.6514 0 0>, rot=<0.0391 -0.0099 -0.2453 -0.9686>}
 hand_r {parent=lowerarm_r, loc=<-22.3491 0 0>, rot=<-0.7081 -0.0507 0.0975 -0.6975>}
 thumb_01_r {parent=hand_r, loc=<-1.3118 0.3191 2.459>, rot=<0.6546 -0.4042 0.0664 -0.6354>}
 thumb_02_r {parent=thumb_01_r, loc=<-5.9336 0.0448 -0.0401>, rot=<-0.0037 0.0213 0.298 -0.9543>}
 thumb_03_r {parent=thumb_02_r, loc=<-2.2223 -0.3682 0.0907>, rot=<0.0118 -0.013 0.0066 -0.9998>}
 index_metacarpal_r {parent=hand_r, loc=<-1.6535 -0.0183 2.96>, rot=<0.0158 -0.0124 -0.0173 -0.9996>}
 index_01_r {parent=index_metacarpal_r, loc=<-8.3008 -0.0357 0.0524>, rot=<-0.0044 0.0485 0.0924 -0.9945>}
 index_02_r {parent=index_01_r, loc=<-3.3587 0.3004 -0.0452>, rot=<-0.002 0.0027 0.3225 -0.9466>}
 index_03_r {parent=index_02_r, loc=<-2.5578 0.123 0.0252>, rot=<-0.002 0.0067 0.2558 -0.9667>}
 middle_metacarpal_r {parent=hand_r, loc=<-2.1856 0.1725 0.5984>, rot=<0.006 0.0168 -0.0329 -0.9993>}
 middle_01_r {parent=middle_metacarpal_r, loc=<-7.3275 0.2447 -0.0129>, rot=<-0.0208 0.0042 0.0858 -0.9961>}
 middle_02_r {parent=middle_01_r, loc=<-4.0123 -0.1401 -0.1175>, rot=<-0.0218 0.0258 0.3512 -0.9357>}
 middle_03_r {parent=middle_02_r, loc=<-2.6077 -0.0534 -0.1373>, rot=<0.0082 0.0188 0.1708 -0.9851>}
 ring_metacarpal_r {parent=hand_r, loc=<-2.2074 0.3625 -1.3638>, rot=<-0.0284 0.0722 -0.0458 -0.9959>}
 ring_01_r {parent=ring_metacarpal_r, loc=<-7.0295 0.6197 0.1446>, rot=<-0.0489 -0.0252 0.1892 -0.9804>}
 ring_02_r {parent=ring_01_r, loc=<-3.6174 -0.4647 0.1027>, rot=<0 0 0.2806 -0.9598>}
 ring_03_r {parent=ring_02_r, loc=<-2.4152 -0.114 -0.0199>, rot=<0 0 0.1189 -0.9929>}
 pinky_metacarpal_r {parent=hand_r, loc=<-1.7616 0.9753 -3.3814>, rot=<-0.0397 0.0814 -0.069 -0.9935>}
 pinky_01_r {parent=pinky_metacarpal_r, loc=<-7.1709 0.4831 0.1094>, rot=<-0.0654 -0.0035 0.2242 -0.9723>}
 pinky_02_r {parent=pinky_01_r, loc=<-2.8274 -0.0893 0.1043>, rot=<0.0037 0.0011 0.2854 -0.9584>}
 pinky_03_r {parent=pinky_02_r, loc=<-2.2046 -0.0307 0.0078>, rot=<-0.0002 0.0098 0.1445 -0.9895>}
 weapon_r {parent=hand_r, loc=<0.1974 3.4111 -0.3811>, rot=<0 0 0 1>}
 hand_attach_r {parent=hand_r, loc=<-8.5001 2.9999 0>, rot=<0 0 -0.7071 -0.7071>}
 deform_wrist_dn_base_r {parent=hand_r, loc=<-0.0004 2.2345 -0.0001>, rot=<0 0 0 -1>}
 deform_wrist_dn_r {parent=deform_wrist_dn_base_r, loc=<0 0 0>, rot=<0 0 0 1>}
 deform_wrist_up_base_r {parent=hand_r, loc=<0.0003 -2.2684 0.0001>, rot=<0 0 0 -1>}
 deform_wrist_up_r {parent=deform_wrist_up_base_r, loc=<0 0 0>, rot=<0 0 0 1>}
 lowerarm_twist_02_r {parent=lowerarm_r, loc=<-9.6199 0 0>, rot=<0 0 0 -1>}
 lowerarm_twist_01_r {parent=lowerarm_r, loc=<-17.8549 0 0>, rot=<0 0 0 -1>}
 deform_elbow_in_base_r {parent=lowerarm_r, loc=<-0.2111 -5.2591 -0.8275>, rot=<0 0 0 -1>}
 deform_elbow_in_r {parent=deform_elbow_in_base_r, loc=<-0.0001 0 0.0002>, rot=<0 0 0 1>}
 deform_elbow_out_base_r {parent=lowerarm_r, loc=<0 4 -0.0001>, rot=<0 0 0 -1>}
 deform_elbow_out_r {parent=deform_elbow_out_base_r, loc=<0 0 0>, rot=<0 0 0 1>}
 upperarm_twist_01_r {parent=upperarm_r, loc=<-0.0041 0 0>, rot=<0 0 0 -1>}
 deform_upperarm_bk_base_r {parent=upperarm_twist_01_r, loc=<-1.0814 7.2192 1.9455>, rot=<0 0 0 1>}
 deform_upperarm_bk_r {parent=deform_upperarm_bk_base_r, loc=<0 0 0>, rot=<0 0 0 1>}
 deform_upperarm_fr_base_r {parent=upperarm_twist_01_r, loc=<0.3689 -7.0506 2.7652>, rot=<0 0 0 1>}
 deform_upperarm_fr_r {parent=deform_upperarm_fr_base_r, loc=<0 0 0>, rot=<0 0 0 1>}
 deform_upperarm_up_base_r {parent=upperarm_twist_01_r, loc=<3.2214 1.1391 -3.3308>, rot=<0 0 0 1>}
 deform_upperarm_up_r {parent=deform_upperarm_up_base_r, loc=<0 0 0>, rot=<0 0 0 -1>}
 upperarm_twist_02_r {parent=upperarm_r, loc=<-13.7941 0 0>, rot=<0 0 0 -1>}
 base_Neck_Jacket_a_l {parent=spine_05, loc=<8.9029 8.4199 -8.82>, rot=<0.0284 -0.7065 -0.0284 -0.7065>}
 dyn_Neck_Jacket_a_l {parent=base_Neck_Jacket_a_l, loc=<0 0 0>, rot=<0 0 0 1>}
 base_Neck_Jacket_b_l {parent=spine_05, loc=<15.2919 -0.0438 -9.2439>, rot=<0.0284 -0.7065 -0.0284 -0.7065>}
 dyn_Neck_Jacket_b_l {parent=base_Neck_Jacket_b_l, loc=<0 0 0>, rot=<0 0 0 1>}
 base_Neck_Jacket_c {parent=spine_05, loc=<15.8476 -4.0266 -0.7152>, rot=<0.0284 -0.7065 -0.0284 -0.7065>}
 dyn_Neck_Jacket_c {parent=base_Neck_Jacket_c, loc=<0 0 0>, rot=<0 0 0 1>}
 base_Neck_Jacket_a_r {parent=spine_05, loc=<8.9029 8.4199 8.82>, rot=<0.0284 -0.7065 -0.0284 -0.7065>}
 dyn_Neck_Jacket_a_r {parent=base_Neck_Jacket_a_r, loc=<0 0 0>, rot=<0 0 0 1>}
 base_Neck_Jacket_b_r {parent=spine_05, loc=<15.2917 -0.0438 9.2439>, rot=<0.0284 -0.7065 -0.0284 -0.7065>}
 dyn_Neck_Jacket_b_r {parent=base_Neck_Jacket_b_r, loc=<0 0 0>, rot=<0 0 0 1>}
 base_necklace {parent=spine_05, loc=<3.3788 10.3315 -0.6313>, rot=<-0.2058 -0.9786 0 0>}
 dyn_necklace_1 {parent=base_necklace, loc=<0 0 0>, rot=<0 0 0 1>}
 dyn_necklace_2 {parent=dyn_necklace_1, loc=<6.0605 0.1861 0.3479>, rot=<0 0 0 1>}
 spine_05_null {parent=spine_04, type=Null, loc=<10.0142 0 0>, rot=<0 0 -0.0015 -1>}
 base_Jacket_a_l {parent=spine_03, loc=<5.8554 8.2846 -13.6683>, rot=<0.0387 -0.706 -0.706 0.0387>}
 dyn_Jacket_a_l {parent=base_Jacket_a_l, loc=<0 0 0>, rot=<0 0 0 1>}
 dyn_Jacket_b_l {parent=dyn_Jacket_a_l, loc=<6.493 0 0>, rot=<0 0 0 1>}
 base_Jacket_b_l {parent=spine_03, loc=<6.9395 -2.1757 -14.9218>, rot=<0.0387 -0.706 -0.706 0.0387>}
 dyn_Jacket_b_a_l {parent=base_Jacket_b_l, loc=<0 0 0>, rot=<0 0 0 1>}
 dyn_Jacket_b_b_l {parent=dyn_Jacket_b_a_l, loc=<6.493 0 0>, rot=<0 0 0 1>}
 base_Jacket_c_l {parent=spine_03, loc=<7.8009 -9.4212 -10.657>, rot=<0.0387 -0.706 -0.706 0.0387>}
 dyn_Jacket_c_a_l {parent=base_Jacket_c_l, loc=<0 0 0>, rot=<0 0 0 1>}
 dyn_Jacket_c_b_l {parent=dyn_Jacket_c_a_l, loc=<6.493 0 0>, rot=<0 0 0 1>}
 base_Jacket_a_r {parent=spine_03, loc=<5.8549 8.2845 13.6683>, rot=<0.0387 -0.706 -0.706 0.0387>}
 dyn_Jacket_a_a_r {parent=base_Jacket_a_r, loc=<0 0 0>, rot=<0 0 0 1>}
 dyn_Jacket_a_b_r {parent=dyn_Jacket_a_a_r, loc=<-6.493 0 0>, rot=<0 0 0 1>}
 dyn_Jacket_a_r {parent=base_Jacket_a_r, loc=<0 0 0>, rot=<0 0 0 1>}
 dyn_Jacket_b_r {parent=dyn_Jacket_a_r, loc=<6.493 0 0>, rot=<0 0 0 1>}
 base_Jacket_b_r {parent=spine_03, loc=<6.9398 -2.1757 14.9218>, rot=<0.0387 -0.706 -0.706 0.0387>}
 dyn_Jacket_b_a_r {parent=base_Jacket_b_r, loc=<0 0 0>, rot=<0 0 0 1>}
 dyn_Jacket_b_b_r {parent=dyn_Jacket_b_a_r, loc=<6.493 0 0>, rot=<0 0 0 1>}
 base_Jacket_c_r {parent=spine_03, loc=<7.8009 -9.4211 10.657>, rot=<0.0387 -0.706 -0.706 0.0387>}
 dyn_Jacket_c_a_r {parent=base_Jacket_c_r, loc=<0 0 0>, rot=<0 0 0 1>}
 dyn_Jacket_c_b_r {parent=dyn_Jacket_c_a_r, loc=<6.493 0 0>, rot=<0 0 0 1>}
 base_Jacket_front_l {parent=spine_03, loc=<5.2493 12.9063 -7.3888>, rot=<-0.0387 0.706 0.706 -0.0387>}
 dyn_Jacket_front_a_l {parent=base_Jacket_front_l, loc=<0 0 0>, rot=<0 0 0 1>}
 dyn_Jacket_front_b_l {parent=dyn_Jacket_front_a_l, loc=<6.493 0 0>, rot=<0 0 0 1>}
 base_Jacket_back {parent=spine_03, loc=<8.0536 -11.975 0.0056>, rot=<0.0547 -0.9985 0 0>}
 dyn_Jacket_a_back {parent=base_Jacket_back, loc=<0 0 0>, rot=<0 0 0 1>}
 dyn_Jacket_b_back {parent=dyn_Jacket_a_back, loc=<6.493 0 0>, rot=<0 0 0 1>}
 base_Jacket_front_r {parent=spine_03, loc=<5.2491 12.9063 7.3888>, rot=<0.0387 -0.706 -0.706 0.0387>}
 dyn_Jacket_front_a_r {parent=base_Jacket_front_r, loc=<0 0 0>, rot=<0 0 0 1>}
 dyn_Jacket_front_b_r {parent=dyn_Jacket_front_a_r, loc=<6.493 0 0>, rot=<0 0 0 1>}
 base_zipper_a {parent=spine_01, loc=<7.0299 12.4448 -3.8609>, rot=<0.0293 -0.7103 -0.7026 0.0291>}
 dyn_zipper_a {parent=base_zipper_a, loc=<0 0 0>, rot=<0 0 0 1>}
 thigh_l {parent=pelvis, loc=<-2.5693 -0.0109 -7.7501>, rot=<0.0012 0.0572 0.0155 -0.9982>}
 calf_l {parent=thigh_l, loc=<-39.0484 0 0>, rot=<0.0006 -0.0161 -0.0391 -0.9991>}
 foot_l {parent=calf_l, loc=<-42.561 0.1077 0.4292>, rot=<0.0418 -0.0455 0.0343 -0.9975>}
 ball_l {parent=foot_l, loc=<-6.1599 13.8264 0.6806>, rot=<0.0001 -0.0001 -0.7063 -0.7079>}
 calf_twist_02_l {parent=calf_l, loc=<-18.8064 -0.0967 0.0397>, rot=<0.0237 0.0095 -0.0035 -0.9997>}
 calf_twist_01_l {parent=calf_l, loc=<-35.813 -0.0967 0.0397>, rot=<0.0237 0.0095 -0.0035 -0.9997>}
 deform_knee_in_base_l {parent=calf_l, loc=<1.298 -5.2957 -0.0496>, rot=<0 0 0 -1>}
 deform_knee_in_l {parent=deform_knee_in_base_l, loc=<0 0 0>, rot=<0 0 0 1>}
 deform_knee_out_base_l {parent=calf_l, loc=<0 4 0>, rot=<0 0 0 -1>}
 deform_knee_out_l {parent=deform_knee_out_base_l, loc=<0 0 0>, rot=<0 0 0 1>}
 thigh_twist_01_l {parent=thigh_l, loc=<-21.6303 0 0>, rot=<0.0013 0 0 -1>}
 base_thigh_chain_a_l {parent=thigh_twist_01_l, loc=<27.8237 8.7189 -0.4394>, rot=<-0.0074 -0.8432 -0.535 -0.0516>}
 off_thigh_chain_a_l {parent=base_thigh_chain_a_l, loc=<0 0 0>, rot=<0 0 0 1>}
 dyn_thigh_chain_a_l {parent=off_thigh_chain_a_l, loc=<0 0 0>, rot=<0 0 0 1>}
 base_thigh_chain_b_l {parent=thigh_twist_01_l, loc=<26.4616 -8.4215 -3.8336>, rot=<-0.0074 -0.8432 -0.535 -0.0516>}
 off_thigh_chain_b_l {parent=base_thigh_chain_b_l, loc=<0 0 0>, rot=<0 0 0 1>}
 dyn_thigh_chain_b_l {parent=off_thigh_chain_b_l, loc=<0 0 0>, rot=<0 0 0 1>}
 base_thigh_chain_ring_l {parent=thigh_twist_01_l, loc=<30.9773 8.8087 -0.4325>, rot=<-0.0074 -0.8432 -0.535 -0.0516>}
 dyn_thigh_chain_ring_l {parent=base_thigh_chain_ring_l, loc=<0 0 0>, rot=<0 0 0 1>}
 deform_groin_base_l {parent=thigh_l, loc=<-2.1908 8.1142 -1.7442>, rot=<0 0 0 -1>}
 deform_groin_l {parent=deform_groin_base_l, loc=<0 0 0>, rot=<0 0 0 1>}
 thigh_r {parent=pelvis, loc=<-2.5693 -0.0109 7.7501>, rot=<0.0572 -0.0012 0.9982 0.0155>}
 calf_r {parent=thigh_r, loc=<39.0484 0 0>, rot=<0.0006 -0.0161 -0.0391 -0.9991>}
 foot_r {parent=calf_r, loc=<42.561 -0.1077 -0.4292>, rot=<0.0418 -0.0455 0.0343 -0.9975>}
 ball_r {parent=foot_r, loc=<6.1599 -13.8264 -0.6806>, rot=<0.0001 -0.0001 -0.7063 -0.7079>}
 calf_twist_02_r {parent=calf_r, loc=<18.8064 0.0967 -0.0397>, rot=<0.0237 0.0095 -0.0035 -0.9997>}
 calf_twist_01_r {parent=calf_r, loc=<35.813 0.0967 -0.0397>, rot=<0.0237 0.0095 -0.0035 -0.9997>}
 deform_knee_in_base_r {parent=calf_r, loc=<-1.298 5.2957 0.0496>, rot=<0 0 0 -1>}
 deform_knee_in_r {parent=deform_knee_in_base_r, loc=<0 0 0>, rot=<0 0 0 1>}
 deform_knee_out_base_r {parent=calf_r, loc=<0 -4 0>, rot=<0 0 0 -1>}
 deform_knee_out_r {parent=deform_knee_out_base_r, loc=<0 0 0>, rot=<0 0 0 1>}
 thigh_twist_01_r {parent=thigh_r, loc=<21.6303 0 0>, rot=<0.0013 0 0 -1>}
 base_thigh_chain_a_r {parent=thigh_twist_01_r, loc=<-28.4907 10.7878 0.0939>, rot=<0.5959 0.06 -0.0753 -0.7973>}
 off_thigh_chain_a_r {parent=base_thigh_chain_a_r, loc=<0 0 0>, rot=<0 0 0 1>}
 dyn_thigh_chain_a_r {parent=off_thigh_chain_a_r, loc=<0 0 0>, rot=<0 0 0 1>}
 base_thigh_chain_b_r {parent=thigh_twist_01_r, loc=<-17.5364 -0.0741 7.6935>, rot=<0.6803 -0.0012 0.034 -0.7322>}
 off_thigh_chain_b_r {parent=base_thigh_chain_b_r, loc=<0 0 0>, rot=<0 0 0 1>}
 dyn_thigh_chain_b_r {parent=off_thigh_chain_b_r, loc=<0 0 0>, rot=<0 0 0 1>}
 deform_groin_base_r {parent=thigh_r, loc=<2.1908 -8.1142 1.7442>, rot=<0 0 0 -1>}
 deform_groin_r {parent=deform_groin_base_r, loc=<0 0 0>, rot=<0 0 0 1>}
 deform_glute_base_l {parent=pelvis, loc=<-8.8064 -8.8222 -8.4287>, rot=<0.0275 0.0501 -0.4472 -0.8926>}
 deform_glute_l {parent=deform_glute_base_l, loc=<0 0 0>, rot=<0 0 0 1>}
 deform_glute_base_r {parent=pelvis, loc=<-8.8062 -8.8222 8.4287>, rot=<0.0501 -0.0275 0.8926 -0.4472>}
 deform_glute_r {parent=deform_glute_base_r, loc=<0 0 0>, rot=<0 0 0 1>}
 attach {parent=root, loc=<0 0 0>, rot=<0 0 0 1>}
 ik_hand_root {parent=root, loc=<0 0 0>, rot=<0 0 0 1>}
 ik_hand_gun {parent=ik_hand_root, loc=<-51.9727 4.0883 102.6935>, rot=<0.6084 0.1793 0.4613 -0.6205>}
 ik_hand_r {parent=ik_hand_gun, loc=<0 0 0>, rot=<0 0 0 1>}
 ik_hand_l {parent=ik_hand_gun, loc=<53.0285 82.1781 35.2051>, rot=<-0.5102 -0.7906 -0.3387 0>}
 ik_foot_root {parent=root, loc=<0 0 0>, rot=<0 0 0 1>}
 ik_foot_l {parent=ik_foot_root, loc=<15.2608 -4.4176 9.4432>, rot=<0.0357 0.7045 0.0284 -0.7082>}
 ik_foot_r {parent=ik_foot_root, loc=<-15.2608 -4.4176 9.4432>, rot=<0.7082 0.0284 -0.7045 0.0357>}
 root_fp {parent=root, loc=<0 0 0>, rot=<0 0 0 1>}
 attach_fp {parent=root_fp, loc=<0 0 0>, rot=<0 0 0 1>}
 camera_root {parent=root_fp, loc=<0 0 0>, rot=<0 0 0 1>}
 camera {parent=camera_root, loc=<0 0 0>, rot=<0 0 0 1>}
 attach_emote_01 {parent=root, loc=<0 0 0>, rot=<0 0 0 1>}
 attach_emote_02 {parent=root, loc=<0 0 0>, rot=<0 0 0 1>}
 attach_emote_03 {parent=root, loc=<0 0 0>, rot=<0 0 0 1>}
 warp_point_01 {parent=root, loc=<0 0 0>, rot=<0 0 0 1>}
 warp_point_02 {parent=root, loc=<0 0 0>, rot=<0 0 0 1>}
 warp_point_03 {parent=root, loc=<0 0 0>, rot=<0 0 0 1>}
 thigh_l_ctrl {controlType=EULER_TRANSFORM, type=Control, loc=<0 0 0>, rot=<0 0 0 1>}
 thigh_r_ctrl {controlType=EULER_TRANSFORM, type=Control, loc=<0 0 0>, rot=<0 0 0 1>}
]
variables [
 PelvisProx {type=double, default=0}
 Pub_IsBaseballSliding {type=double, default=0}
 Pub_IsCrouching {type=double, default=0}
 Pub_IsKneeSliding {type=double, default=0}
 Pub_IsPickaxeSwing {type=double, default=0}
 SPR L Leg Bkd {type=double, default=0}
 SPR L Leg Fwd {type=double, default=0}
 SPR L Leg Otw {type=double, default=0}
 SPR R Leg Bkd {type=double, default=0}
 SPR R Leg Fwd {type=double, default=0}
 SPR R Leg Otw {type=double, default=0}
]
