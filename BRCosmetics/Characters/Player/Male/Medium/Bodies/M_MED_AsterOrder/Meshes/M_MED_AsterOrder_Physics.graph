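# Physics asset for the M_MED_AsterOrder male medium player body, covering the spine, jacket and zipper bones.
# Four spine capsules form the torso, linked by swing and twist limited constraints.
# Nine jacket chains (front, back, sides) hang from spine_03, each a base capsule plus two dyn capsules; a shorter zipper chain hangs from spine_01.
# Jacket and zipper joints have zero twist limit, with swing limits varying by panel.
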PhysicsAsset /BRCosmetics/Characters/Player/Male/Medium/Bodies/M_MED_AsterOrder/Meshes/M_MED_AsterOrder_Physics
bodies {
 base_Jacket_a_l [{shapeType=Capsule, length=4, radius=3, shapeName=base_Jacket_a_l_capsule, center=<1 0 0>, rotation=<rot 90 0 0>}]
 base_Jacket_a_r [{shapeType=Capsule, length=4, radius=3, shapeName=base_Jacket_a_r_capsule, center=<1 0 0>, rotation=<rot 90 0 0>}]
 base_Jacket_b_l [{shapeType=Capsule, length=4, radius=3, shapeName=base_Jacket_b_l_capsule, center=<1 0 0>, rotation=<rot 90 0 0>}]
 base_Jacket_b_r [{shapeType=Capsule, length=4, radius=3, shapeName=base_Jacket_b_r_capsule, center=<1 0 0>, rotation=<rot 90 0 0>}]
 base_Jacket_back [{shapeType=Capsule, length=4, radius=3, shapeName=base_Jacket_back_capsule, center=<1 0 0>, rotation=<rot 90 0 0>}]
 base_Jacket_c_l [{shapeType=Capsule, length=4, radius=3, shapeName=base_Jacket_c_l_capsule, center=<1 0 0>, rotation=<rot 90 0 0>}]
 base_Jacket_c_r [{shapeType=Capsule, length=4, radius=3, shapeName=base_Jacket_c_r_capsule, center=<1 0 0>, rotation=<rot 90 0 0>}]
 base_Jacket_front_l [{shapeType=Capsule, length=4, radius=3, shapeName=base_Jacket_front_l_capsule, center=<1 0 0>, rotation=<rot 90 0 0>}]
 base_Jacket_front_r [{shapeType=Capsule, length=4, radius=3, shapeName=base_Jacket_front_r_capsule, center=<1 0 0>, rotation=<rot 90 0 0>}]
 base_zipper_a [{shapeType=Capsule, length=0.505, radius=0.505, shapeName=base_zipper_a_capsule, rotation=<rot 0 -90 0>}]
 dyn_Jacket_a_back [{shapeType=Capsule, length=4, radius=2.5, shapeName=dyn_Jacket_a_back_capsule, center=<0.85 0 0>, rotation=<rot 90 0 0>}]
 dyn_Jacket_a_l [{shapeType=Capsule, length=4, radius=2.5, shapeName=dyn_Jacket_a_l_capsule, center=<0.85 0 0>, rotation=<rot 90 0 0>}]
 dyn_Jacket_a_r [{shapeType=Capsule, length=4, radius=2.5, shapeName=dyn_Jacket_a_r_capsule, center=<0.85 0 0>, rotation=<rot 90 0 0>}]
 dyn_Jacket_b_a_l [{shapeType=Capsule, length=4, radius=2.5, shapeName=dyn_Jacket_b_a_l_capsule, center=<0.85 0 0>, rotation=<rot 90 0 0>}]
 dyn_Jacket_b_a_r [{shapeType=Capsule, length=4, radius=2.5, shapeName=dyn_Jacket_b_a_r_capsule, center=<0.85 0 0>, rotation=<rot 90 0 0>}]
 dyn_Jacket_b_b_l [{shapeType=Capsule, length=4, radius=2, shapeName=dyn_Jacket_b_b_l_capsule, center=<1 0 0>, rotation=<rot 90 0 0>}]
 dyn_Jacket_b_b_r [{shapeType=Capsule, length=4, radius=2, shapeName=dyn_Jacket_b_b_r_capsule, center=<1 0 0>, rotation=<rot 90 0 0>}]
 dyn_Jacket_b_back [{shapeType=Capsule, length=4, radius=2, shapeName=dyn_Jacket_b_back_capsule, center=<1 0 0>, rotation=<rot 90 0 0>}]
 dyn_Jacket_b_l [{shapeType=Capsule, length=4, radius=2, shapeName=dyn_Jacket_b_l_capsule, center=<1 0 0>, rotation=<rot 90 0 0>}]
 dyn_Jacket_b_r [{shapeType=Capsule, length=4, radius=2, shapeName=dyn_Jacket_b_r_capsule, center=<1 0 0>, rotation=<rot 90 0 0>}]
 dyn_Jacket_c_a_l [{shapeType=Capsule, length=4, radius=2.5, shapeName=dyn_Jacket_c_a_l_capsule, center=<0.85 0 0>, rotation=<rot 90 0 0>}]
 dyn_Jacket_c_a_r [{shapeType=Capsule, length=4, radius=2.5, shapeName=dyn_Jacket_c_a_r_capsule, center=<0.85 0 0>, rotation=<rot 90 0 0>}]
 dyn_Jacket_c_b_l [{shapeType=Capsule, length=4, radius=2, shapeName=dyn_Jacket_c_b_l_capsule, center=<1 0 0>, rotation=<rot 90 0 0>}]
 dyn_Jacket_c_b_r [{shapeType=Capsule, length=4, radius=2, shapeName=dyn_Jacket_c_b_r_capsule, center=<1 0 0>, rotation=<rot 90 0 0>}]
 dyn_Jacket_front_a_l [{shapeType=Capsule, length=4, radius=2.5, shapeName=dyn_Jacket_front_a_l_capsule, center=<0.85 0 0>, rotation=<rot 90 0 0>}]
 dyn_Jacket_front_a_r [{shapeType=Capsule, length=4, radius=2.5, shapeName=dyn_Jacket_front_a_r_capsule, center=<0.85 0 0>, rotation=<rot 90 0 0>}]
 dyn_Jacket_front_b_l [{shapeType=Capsule, length=4, radius=2, shapeName=dyn_Jacket_front_b_l_capsule, center=<1 0 0>, rotation=<rot 90 0 0>}]
 dyn_Jacket_front_b_r [{shapeType=Capsule, length=4, radius=2, shapeName=dyn_Jacket_front_b_r_capsule, center=<1 0 0>, rotation=<rot 90 0 0>}]
 dyn_zipper_a [{shapeType=Capsule, length=1, radius=1, shapeName=dyn_zipper_a_capsule, center=<1 0 0>, rotation=<rot 90 0 0>}]
 spine_01 [{shapeType=Capsule, length=4.5902, radius=10.726, center=<8.1325 0.2801 0.0136>, rotation=<rot 0.1028 -90 0.2061>}]
 spine_03 [{shapeType=Capsule, length=4.4821, radius=10.7509, center=<0.5611 0.5082 -0.0162>, rotation=<rot 1.1391 -90 -0.2841>}]
 spine_04 [{shapeType=Capsule, length=7.9801, radius=14.2863, center=<4.2393 0.2319 -0.5916>, rotation=<rot -5.4651 -90 -0.5599>}]
 spine_05 [{shapeType=Capsule, length=18.8884, radius=10.0431}]
}
constraints [
 {bone1Name=spine_03, bone2Name=spine_01, swing1LimitDegrees=45, swing2LimitDegrees=45, twistLimitDegrees=45}
 {bone1Name=spine_04, bone2Name=spine_03, swing1LimitDegrees=45, swing2LimitDegrees=45, twistLimitDegrees=45}
 {bone1Name=spine_05, bone2Name=spine_04, swing1LimitDegrees=45, swing2LimitDegrees=45, twistLimitDegrees=45}
 {bone1Name=base_Jacket_a_l, bone2Name=spine_03, swing1LimitDegrees=20, swing2LimitDegrees=20, twistLimitDegrees=0}
 {bone1Name=dyn_Jacket_a_l, bone2Name=base_Jacket_a_l, swing1LimitDegrees=10, swing2LimitDegrees=15, twistLimitDegrees=0}
 {bone1Name=dyn_Jacket_b_l, bone2Name=dyn_Jacket_a_l, swing1LimitDegrees=15, swing2LimitDegrees=15, twistLimitDegrees=0}
 {bone1Name=base_Jacket_a_r, bone2Name=spine_03, swing1LimitDegrees=20, swing2LimitDegrees=20, twistLimitDegrees=0}
 {bone1Name=dyn_Jacket_a_r, bone2Name=base_Jacket_a_r, swing1LimitDegrees=10, swing2LimitDegrees=15, twistLimitDegrees=0}
 {bone1Name=dyn_Jacket_b_r, bone2Name=dyn_Jacket_a_r, swing1LimitDegrees=15, swing2LimitDegrees=15, twistLimitDegrees=0}
 {bone1Name=base_Jacket_b_l, bone2Name=spine_03, swing1LimitDegrees=20, swing2LimitDegrees=20, twistLimitDegrees=0}
 {bone1Name=dyn_Jacket_b_a_l, bone2Name=base_Jacket_b_l, swing1LimitDegrees=10, swing2LimitDegrees=15, twistLimitDegrees=0}
 {bone1Name=dyn_Jacket_b_b_l, bone2Name=dyn_Jacket_b_a_l, swing1LimitDegrees=15, swing2LimitDegrees=15, twistLimitDegrees=0}
 {bone1Name=base_Jacket_b_r, bone2Name=spine_03, swing1LimitDegrees=20, swing2LimitDegrees=20, twistLimitDegrees=0}
 {bone1Name=dyn_Jacket_b_a_r, bone2Name=base_Jacket_b_r, swing1LimitDegrees=10, swing2LimitDegrees=15, twistLimitDegrees=0}
 {bone1Name=dyn_Jacket_b_b_r, bone2Name=dyn_Jacket_b_a_r, swing1LimitDegrees=15, swing2LimitDegrees=15, twistLimitDegrees=0}
 {bone1Name=base_Jacket_back, bone2Name=spine_03, swing1LimitDegrees=20, swing2LimitDegrees=20, twistLimitDegrees=0}
 {bone1Name=dyn_Jacket_a_back, bone2Name=base_Jacket_back, swing1LimitDegrees=20, swing2LimitDegrees=10, twistLimitDegrees=0}
 {bone1Name=dyn_Jacket_b_back, bone2Name=dyn_Jacket_a_back, swing1LimitDegrees=25, swing2LimitDegrees=10, twistLimitDegrees=0}
 {bone1Name=base_Jacket_c_l, bone2Name=spine_03, swing1LimitDegrees=20, swing2LimitDegrees=20, twistLimitDegrees=0}
 {bone1Name=dyn_Jacket_c_a_l, bone2Name=base_Jacket_c_l, swing1LimitDegrees=10, swing2LimitDegrees=25, twistLimitDegrees=0}
 {bone1Name=dyn_Jacket_c_b_l, bone2Name=dyn_Jacket_c_a_l, swing1LimitDegrees=10, swing2LimitDegrees=25, twistLimitDegrees=0}
 {bone1Name=base_Jacket_c_r, bone2Name=spine_03, swing1LimitDegrees=20, swing2LimitDegrees=20, twistLimitDegrees=0}
 {bone1Name=dyn_Jacket_c_a_r, bone2Name=base_Jacket_c_r, swing1LimitDegrees=10, swing2LimitDegrees=25, twistLimitDegrees=0}
 {bone1Name=dyn_Jacket_c_b_r, bone2Name=dyn_Jacket_c_a_r, swing1LimitDegrees=10, swing2LimitDegrees=25, twistLimitDegrees=0}
 {bone1Name=base_Jacket_front_l, bone2Name=spine_03, swing1LimitDegrees=20, swing2LimitDegrees=20, twistLimitDegrees=0}
 {bone1Name=dyn_Jacket_front_a_l, bone2Name=base_Jacket_front_l, swing1LimitDegrees=20, swing2LimitDegrees=15, twistLimitDegrees=0}
 {bone1Name=dyn_Jacket_front_b_l, bone2Name=dyn_Jacket_front_a_l, swing1LimitDegrees=25, swing2LimitDegrees=15, twistLimitDegrees=0}
 {bone1Name=base_Jacket_front_r, bone2Name=spine_03, swing1LimitDegrees=20, swing2LimitDegrees=20, twistLimitDegrees=0}
 {bone1Name=dyn_Jacket_front_a_r, bone2Name=base_Jacket_front_r, swing1LimitDegrees=20, swing2LimitDegrees=15, twistLimitDegrees=0}
 {bone1Name=dyn_Jacket_front_b_r, bone2Name=dyn_Jacket_front_a_r, swing1LimitDegrees=25, swing2LimitDegrees=15, twistLimitDegrees=0}
 {bone1Name=base_zipper_a, bone2Name=spine_01, swing1LimitDegrees=45, swing2LimitDegrees=45, twistLimitDegrees=0}
 {bone1Name=dyn_zipper_a, bone2Name=base_zipper_a, swing1LimitDegrees=30, swing2LimitDegrees=30, twistLimitDegrees=0}
]
