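# Physics asset for the necklace on the medium male AsterOrder player cosmetic body, defining capsule bodies and joints for its dangling motion.
# Two dynamic necklace bones chain from a small base capsule, which attaches to spine_05, with large spine_04 and spine_05 capsules forming the upper-torso volume.
# Tight swing and twist limits keep the two necklace links nearly rigid, while the base has wider swing and the spine-to-spine joint rotates freely.
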PhysicsAsset /BRCosmetics/Characters/Player/Male/Medium/Bodies/M_MED_AsterOrder/Meshes/M_MED_AsterOrder_Necklace_Physics
bodies {
 base_necklace [{shapeType=Capsule, length=0.505, radius=0.505, shapeName=base_necklace_capsule, rotation=<rot 0 -90 0>}]
 dyn_necklace_1 [{
   shapeType=Capsule
   length=3.4412
   radius=1.6572
   shapeName=dyn_necklace_1_capsule
   center=<4.6412 0.2028 0.3759>
   rotation=<rot -4.3137 -90 0.2075>
  }]
 dyn_necklace_2 [{
   shapeType=Capsule
   length=5
   radius=1.5
   shapeName=dyn_necklace_2_capsule
   center=<2.2845 -0.5889 0.0301>
   rotation=<rot 0 -106.5167 -89.395>
  }]
 spine_04 [{shapeType=Capsule, length=11.9889, radius=10.0111, shapeName=spine_05_capsule, center=<-2.5735 3.6032 0>, rotation=<rot 0 -15.0669 0>}]
 spine_05 [{shapeType=Capsule, length=10.1921, radius=11.8079, shapeName=spine_05_capsule, center=<-7.247 0.2378 0>, rotation=<rot 0 -15.0669 0>}]
}
constraints [
 {bone1Name=base_necklace, bone2Name=spine_05, swing1LimitDegrees=45, swing2LimitDegrees=45, twistLimitDegrees=0}
 {bone1Name=dyn_necklace_1, bone2Name=base_necklace, swing1LimitDegrees=10, swing2LimitDegrees=3, twistLimitDegrees=1}
 {bone1Name=dyn_necklace_2, bone2Name=dyn_necklace_1, swing1LimitDegrees=10, swing2LimitDegrees=2, twistLimitDegrees=1}
 {
  bone1Name=spine_05
  bone2Name=spine_04
  swing1LimitDegrees=45
  swing1Motion=Free
  swing2LimitDegrees=45
  swing2Motion=Free
  twistLimitDegrees=45
  twistMotion=Free
 }
]
